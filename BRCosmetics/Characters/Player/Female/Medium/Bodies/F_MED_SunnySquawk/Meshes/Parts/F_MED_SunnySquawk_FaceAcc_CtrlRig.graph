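# Control Rig for the SunnySquawk female medium cosmetic FaceAcc part, posing dfrm_ deformation bones.
# The visible main graph constrains thighs to source bones, then solves stretchy two-bone leg IK toward foot controls.
# A spherical pose reader on the right thigh smoothly shifts that leg's pole-vector bone.
# Library functions define toggleable elbow, knee, groin/glute and upper-arm correctives, weighting helper bones by joint angles or pose readers, with mirror utilities.
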ControlRigBlueprint /BRCosmetics/Characters/Player/Female/Medium/Bodies/F_MED_SunnySquawk/Meshes/Parts/F_MED_SunnySquawk_FaceAcc_CtrlRig
graphs [
 ContainedGraph {nodes [
   n1 = RigVMFunctionEntryNode()
   SequenceAggregate_2 = &RigVMFunction_Sequence(ExecuteContext=@n1.ExecuteContext)
   SequenceAggregate_2_1 = &RigVMFunction_Sequence(ExecuteContext=@SequenceAggregate_2.B)
   RigVMFunctionReturnNode(A=@SequenceAggregate_2.A, B=@SequenceAggregate_2_1.A, C=@SequenceAggregate_2_1.B)
  ]}
 ContainedGraph {nodes [
   n1 = RigVMFunctionEntryNode()
   n2 = &RigUnit_GetTransform(Space=LocalSpace, Item=@n1.Item)
   n3 = &RigUnit_GetTransform(Space=LocalSpace, bInitial=true, Item=@n1.Item)
   > RigVMFunctionReturnNode(OutTransform=/Script/RigVM.RigVMFunction_MathTransformToEulerTransform(Value=&RigVMFunction_MathTransformMakeRelative(Global=@n2.Transform, Parent=@n3.Transform).Local) {is=struct RigVMUnitNode, template="Cast::Execute(in Value,out Result)"}.Result)
  ]}
 ContainedGraph {nodes [
   n1 = RigVMFunctionEntryNode()
   n2 = &RigVMFunction_MathTransformToEulerTransform(Value=@n1.transform) {template="Cast::Execute(in Value,out Result)"}
   > RigVMFunctionReturnNode(
    mirrorTransform=(Rotation=(X=0,Y=0,Z=0,W=1),Translation=(X=0,Y=0,Z=0),Scale3D=(X=0,Y=0,Z=0))
    mirrorTransform.Rotation=/Script/RigVM.RigVMFunction_MathQuaternionFromRotatorV2(Value=@n2.Result.Rotation) {is=struct RigVMUnitNode, template="Cast::Execute(in Value,out Result)"}.Result
    mirrorTransform.Translation=/Script/RigVM.RigVMFunction_MathVectorScale(Factor=-1, Value=@n2.Result.Location) {is=struct RigVMUnitNode, template="Scale::Execute(in Value,in Factor,out Result)"}.Result
    mirrorTransform.Scale3D=@n2.Result.Scale
   )
  ]}
 ContainedGraph {nodes [
   n1 = RigVMFunctionEntryNode()
   > SequenceExecution_2 = RigVMAggregateNode()
   GetDeltaTransformFromRefPose_FNC = RigVMFunctionReferenceNode(Item=(Type=Bone,Name="dfrm_calf_l"), ExecuteContext=@SequenceExecution_2.A)
   GetDeltaTransformFromRefPose_FNC_1 = RigVMFunctionReferenceNode(Item=(Type=Bone,Name="dfrm_calf_r"), ExecuteContext=@SequenceExecution_2.B)
   SequenceExecution = RigVMAggregateNode(ExecuteContext=@SequenceExecution_2.C)
   n2 = &RigVMFunction_ControlFlowBranch(Condition=@n1.knee_in, ExecuteContext=@SequenceExecution.A)
   SequenceExecution_1 = RigVMAggregateNode(ExecuteContext=@n2.True)
   &RigUnit_ParentConstraint(
    AdvancedSettings=(RotationOrderForFilter=XZY)
    Child=(Type=Bone,Name="deform_knee_in_base_l")
    Parents=((Item=(Type=Bone,Name="dfrm_calf_l"),Weight=0.5),(Item=(Type=Bone,Name="dfrm_thigh_l"),Weight=0.5))
    Weight=1
    ExecutePin=@SequenceExecution_1.A
   )
   n3 = &RigVMFunction_MathFloatRemap(SourceMaximum=100, SourceMinimum=10, bClamp=true, Value=@GetDeltaTransformFromRefPose_FNC.OutTransform.Rotation.Yaw)
   > &RigUnit_ModifyTransforms(
    ItemToModify=((Item=(Type=Bone,Name="deform_knee_in_l"),Transform=(Rotation=(X=0,Y=0,Z=0,W=1),Translation=(X=0,Y=0,Z=0),Scale3D=(X=1,Y=1,Z=1))))
    Weight=@n3.Result
   )
   &RigUnit_ParentConstraint(
    AdvancedSettings=(RotationOrderForFilter=XZY)
    Child=(Type=Bone,Name="deform_knee_in_base_r")
    Parents=((Item=(Type=Bone,Name="dfrm_calf_r"),Weight=0.5),(Item=(Type=Bone,Name="dfrm_thigh_r"),Weight=0.5))
    Weight=1
    ExecutePin=@SequenceExecution_1.B
   )
   > Deform_GetMirrorTransform_FNC = RigVMFunctionReferenceNode(transform=(Rotation=(X=0,Y=0,Z=0,W=1),Translation=(X=0,Y=0,Z=0),Scale3D=(X=1,Y=1,Z=1)))
   n4 = &RigVMFunction_MathFloatRemap(SourceMaximum=100, SourceMinimum=10, bClamp=true, Value=@GetDeltaTransformFromRefPose_FNC_1.OutTransform.Rotation.Yaw)
   > &RigUnit_ModifyTransforms(
    ItemToModify=((Item=(Type=Bone,Name="deform_knee_in_r"),Transform=(Rotation=(X=0,Y=0,Z=0,W=1),Translation=(X=0,Y=0,Z=0),Scale3D=(X=1,Y=1,Z=1))))
    ItemToModify.0.Transform=@Deform_GetMirrorTransform_FNC.mirrorTransform
    Weight=@n4.Result
   )
   n5 = &RigVMFunction_ControlFlowBranch(Condition=@n1.knee_out, ExecuteContext=@SequenceExecution.B)
   SequenceExecution_1_1 = RigVMAggregateNode(ExecuteContext=@n5.True)
   &RigUnit_ParentConstraint(
    AdvancedSettings=(RotationOrderForFilter=XZY)
    Child=(Type=Bone,Name="deform_knee_out_base_l")
    Parents=((Item=(Type=Bone,Name="dfrm_calf_l"),Weight=0.1),(Item=(Type=Bone,Name="dfrm_thigh_l"),Weight=0.9))
    Weight=1
    ExecutePin=@SequenceExecution_1_1.A
   )
   > &RigUnit_ModifyTransforms(
    ItemToModify=((Item=(Type=Bone,Name="deform_knee_out_l"),Transform=(Rotation=(X=0,Y=0,Z=0,W=1),Translation=(X=0,Y=0,Z=0),Scale3D=(X=1,Y=1,Z=1))))
    Weight=@n3.Result
   )
   &RigUnit_ParentConstraint(
    AdvancedSettings=(RotationOrderForFilter=XZY)
    Child=(Type=Bone,Name="deform_knee_out_base_r")
    Parents=((Item=(Type=Bone,Name="dfrm_calf_r"),Weight=0.1),(Item=(Type=Bone,Name="dfrm_thigh_r"),Weight=0.9))
    Weight=1
    ExecutePin=@SequenceExecution_1_1.B
   )
   > Deform_GetMirrorTransform_FNC_1 = RigVMFunctionReferenceNode(transform=(Rotation=(X=0,Y=0,Z=0,W=1),Translation=(X=0,Y=0,Z=0),Scale3D=(X=1,Y=1,Z=1)))
   > &RigUnit_ModifyTransforms(
    ItemToModify=((Item=(Type=Bone,Name="deform_knee_out_r"),Transform=(Rotation=(X=0,Y=0,Z=0,W=1),Translation=(X=0,Y=0,Z=0),Scale3D=(X=1,Y=1,Z=1))))
    ItemToModify.0.Transform=@Deform_GetMirrorTransform_FNC_1.mirrorTransform
    Weight=@n4.Result
   )
   RigVMFunctionReturnNode(ExecuteContext=@SequenceExecution_2.D)
  ]}
 ContainedGraph {nodes [
   n1 = RigVMFunctionEntryNode()
   SequenceExecution_2 = &RigVMFunction_Sequence(ExecuteContext=@n1.ExecuteContext)
   SequenceExecution_2_1 = &RigVMFunction_Sequence(ExecuteContext=@SequenceExecution_2.B)
   SequenceExecution_2_2 = &RigVMFunction_Sequence(ExecuteContext=@SequenceExecution_2_1.B)
   RigVMFunctionReturnNode(A=@SequenceExecution_2.A, B=@SequenceExecution_2_1.A, C=@SequenceExecution_2_2.A, D=@SequenceExecution_2_2.B)
  ]}
 ContainedGraph {nodes [
   n1 = RigVMFunctionEntryNode()
   SequenceExecution = &RigVMFunction_Sequence(ExecuteContext=@n1.ExecuteContext)
   SequenceExecution_1 = &RigVMFunction_Sequence(ExecuteContext=@SequenceExecution.B)
   SequenceExecution_2 = &RigVMFunction_Sequence(ExecuteContext=@SequenceExecution_1.B)
   RigVMFunctionReturnNode(A=@SequenceExecution.A, B=@SequenceExecution_1.A, C=@SequenceExecution_2.A, D=@SequenceExecution_2.B)
  ]}
 ContainedGraph {nodes [
   n1 = RigVMFunctionEntryNode()
   SequenceExecution_1 = &RigVMFunction_Sequence(ExecuteContext=@n1.ExecuteContext)
   SequenceExecution_1_1 = &RigVMFunction_Sequence(ExecuteContext=@SequenceExecution_1.B)
   SequenceExecution_1_2 = &RigVMFunction_Sequence(ExecuteContext=@SequenceExecution_1_1.B)
   RigVMFunctionReturnNode(A=@SequenceExecution_1.A, B=@SequenceExecution_1_1.A, C=@SequenceExecution_1_2.A, D=@SequenceExecution_1_2.B)
  ]}
 ContainedGraph {nodes [
   n1 = RigVMFunctionEntryNode()
   SequenceExecution_1_1 = &RigVMFunction_Sequence(ExecuteContext=@n1.ExecuteContext)
   SequenceExecution_1_1_1 = &RigVMFunction_Sequence(ExecuteContext=@SequenceExecution_1_1.B)
   SequenceExecution_1_1_2 = &RigVMFunction_Sequence(ExecuteContext=@SequenceExecution_1_1_1.B)
   RigVMFunctionReturnNode(A=@SequenceExecution_1_1.A, B=@SequenceExecution_1_1_1.A, C=@SequenceExecution_1_1_2.A, D=@SequenceExecution_1_1_2.B)
  ]}
 ContainedGraph {nodes [
   n1 = RigVMFunctionEntryNode()
   n2 = &RigVMFunction_MathTransformToEulerTransform(Value=@n1.transform) {template="Cast::Execute(in Value,out Result)"}
   > RigVMFunctionReturnNode(
    mirrorTransform=(Rotation=(X=0,Y=0,Z=0,W=1),Translation=(X=0,Y=0,Z=0),Scale3D=(X=0,Y=0,Z=0))
    mirrorTransform.Rotation=&RigVMFunction_MathQuaternionFromRotatorV2(Value=@n2.Result.Rotation) {template="Cast::Execute(in Value,out Result)"}.Result
    mirrorTransform.Translation=&RigVMFunction_MathVectorScale(Factor=-1, Value=@n2.Result.Location) {template="Scale::Execute(in Value,in Factor,out Result)"}.Result
    mirrorTransform.Scale3D=@n2.Result.Scale
   )
  ]}
 ContainedGraph {nodes [
   n1 = RigVMFunctionEntryNode()
   n2 = &RigVMFunction_ControlFlowBranch(ExecuteContext=@n1.ExecuteContext, Condition=&RigVMFunction_MathBoolOr(B=@n1.glute, A=@n1.groin).Result)
   RigVMFunctionReturnNode(ExecuteContext=@n2.Completed)
   SequenceExecution = RigVMAggregateNode(ExecuteContext=@n2.True)
   n3 = &RigUnit_SphericalPoseReader(
    ActiveRegionScaleFactors=(PositiveWidth=3.5,NegativeWidth=0.5,PositiveHeight=3,NegativeHeight=3)
    ActiveRegionSize=0.1
    Debug=(bDrawDebug=false,DebugScale=25,DebugSegments=20,DebugThickness=0.25)
    DriverAxis=(X=-1,Y=0,Z=0)
    DriverItem=(Type=Bone,Name="dfrm_thigh_l")
    FalloffRegionScaleFactors=(PositiveWidth=1,NegativeWidth=1,PositiveHeight=1,NegativeHeight=1)
    FalloffSize=0.45
    OptionalParentItem=(Type=Bone)
    RotationOffset=(X=90,Y=0,Z=0)
    ExecutePin=@SequenceExecution.A
    Debug.bDrawDebug=@n1.debug
   )
   > n4 = &RigUnit_SphericalPoseReader(
    ActiveRegionScaleFactors=(PositiveWidth=3.5,NegativeWidth=0.5,PositiveHeight=3,NegativeHeight=3)
    ActiveRegionSize=0.1
    Debug=(bDrawDebug=false,DebugScale=25,DebugSegments=20,DebugThickness=0.25)
    DriverAxis=(X=1,Y=0,Z=0)
    DriverItem=(Type=Bone,Name="dfrm_thigh_r")
    FalloffRegionScaleFactors=(PositiveWidth=1,NegativeWidth=1,PositiveHeight=1,NegativeHeight=1)
    FalloffSize=0.45
    FlipHeightScaling=false
    FlipWidthScaling=true
    OptionalParentItem=(Type=Bone)
    RotationOffset=(X=-90,Y=0,Z=0)
    Debug.bDrawDebug=@n1.debug
   )
   n5 = &RigVMFunction_ControlFlowBranch(ExecuteContext=@SequenceExecution.B, Condition=@n1.groin)
   SequenceExecution_1 = RigVMAggregateNode(ExecuteContext=@n5.True)
   &RigUnit_ParentConstraint(
    AdvancedSettings=(RotationOrderForFilter=XZY)
    Child=(Type=Bone,Name="deform_groin_base_l")
    Parents=((Item=(Type=Bone,Name="dfrm_thigh_l"),Weight=0.7),(Item=(Type=Bone,Name="dfrm_pelvis"),Weight=0.3))
    Weight=1
    ExecutePin=@SequenceExecution_1.A
   )
   > &RigUnit_ModifyTransforms(
    ItemToModify=((Item=(Type=Bone,Name="deform_groin_l"),Transform=(Rotation=(X=0,Y=0,Z=0,W=1),Translation=(X=-3,Y=12,Z=0),Scale3D=(X=1,Y=1,Z=1))))
    Weight=@n3.OutputParam
   )
   &RigUnit_ParentConstraint(
    AdvancedSettings=(RotationOrderForFilter=XZY)
    Child=(Type=Bone,Name="deform_groin_base_r")
    Parents=((Item=(Type=Bone,Name="dfrm_thigh_r"),Weight=0.7),(Item=(Type=Bone,Name="dfrm_pelvis"),Weight=0.3))
    Weight=1
    ExecutePin=@SequenceExecution_1.B
   )
   > &RigUnit_ModifyTransforms(
    ItemToModify=((Item=(Type=Bone,Name="deform_groin_r"),Transform=(Rotation=(X=0,Y=0,Z=0,W=1),Translation=(X=3,Y=-12,Z=0),Scale3D=(X=1,Y=1,Z=1))))
    Weight=@n4.OutputParam
   )
   n6 = &RigVMFunction_ControlFlowBranch(ExecuteContext=@SequenceExecution.C, Condition=@n1.glute)
   SequenceExecution_1_1 = RigVMAggregateNode(ExecuteContext=@n6.True)
   &RigUnit_ParentConstraint(
    AdvancedSettings=(RotationOrderForFilter=XZY)
    Child=(Type=Bone,Name="deform_glute_base_l")
    Parents=((Item=(Type=Bone,Name="dfrm_thigh_l"),Weight=0.25),(Item=(Type=Bone,Name="dfrm_pelvis"),Weight=0.75))
    Weight=1
    ExecutePin=@SequenceExecution_1_1.A
   )
   > &RigUnit_ModifyTransforms(
    ItemToModify=((Item=(Type=Bone,Name="deform_glute_l"),Transform=(Rotation=(X=0,Y=0,Z=0,W=1),Translation=(X=0,Y=0,Z=0),Scale3D=(X=1,Y=1,Z=1))))
    Weight=@n3.OutputParam
   )
   &RigUnit_ParentConstraint(
    AdvancedSettings=(RotationOrderForFilter=XZY)
    Child=(Type=Bone,Name="deform_glute_base_r")
    Parents=((Item=(Type=Bone,Name="dfrm_thigh_r"),Weight=0.25),(Item=(Type=Bone,Name="dfrm_pelvis"),Weight=0.75))
    Weight=1
    ExecutePin=@SequenceExecution_1_1.B
   )
   > Deform_GetMirrorTransform_FNC_1 = RigVMFunctionReferenceNode(transform=(Rotation=(X=0,Y=0,Z=0,W=1),Translation=(X=0,Y=0,Z=0),Scale3D=(X=1,Y=1,Z=1)))
   > &RigUnit_ModifyTransforms(
    ItemToModify=((Item=(Type=Bone,Name="deform_glute_r"),Transform=(Rotation=(X=0,Y=0,Z=0,W=1),Translation=(X=0,Y=0,Z=0),Scale3D=(X=1,Y=1,Z=1))))
    ItemToModify.0.Transform=@Deform_GetMirrorTransform_FNC_1.mirrorTransform
    Weight=@n4.OutputParam
   )
  ]}
 ContainedGraph {nodes [
   n1 = RigVMFunctionEntryNode()
   SequenceExecution = &RigVMFunction_Sequence(ExecuteContext=@n1.ExecuteContext)
   SequenceExecution_1 = &RigVMFunction_Sequence(ExecuteContext=@SequenceExecution.B)
   SequenceExecution_2 = &RigVMFunction_Sequence(ExecuteContext=@SequenceExecution_1.B)
   RigVMFunctionReturnNode(A=@SequenceExecution.A, B=@SequenceExecution_1.A, C=@SequenceExecution_2.A, D=@SequenceExecution_2.B)
  ]}
 ContainedGraph {nodes [
   n1 = RigVMFunctionEntryNode()
   SequenceExecution_1 = &RigVMFunction_Sequence(ExecuteContext=@n1.ExecuteContext)
   SequenceExecution_1_1 = &RigVMFunction_Sequence(ExecuteContext=@SequenceExecution_1.B)
   SequenceExecution_1_2 = &RigVMFunction_Sequence(ExecuteContext=@SequenceExecution_1_1.B)
   RigVMFunctionReturnNode(A=@SequenceExecution_1.A, B=@SequenceExecution_1_1.A, C=@SequenceExecution_1_2.A, D=@SequenceExecution_1_2.B)
  ]}
 ContainedGraph {nodes [
   n1 = RigVMFunctionEntryNode()
   SequenceExecution_1_1 = &RigVMFunction_Sequence(ExecuteContext=@n1.ExecuteContext)
   SequenceExecution_1_1_1 = &RigVMFunction_Sequence(ExecuteContext=@SequenceExecution_1_1.B)
   SequenceExecution_1_1_2 = &RigVMFunction_Sequence(ExecuteContext=@SequenceExecution_1_1_1.B)
   RigVMFunctionReturnNode(A=@SequenceExecution_1_1.A, B=@SequenceExecution_1_1_1.A, C=@SequenceExecution_1_1_2.A, D=@SequenceExecution_1_1_2.B)
  ]}
 ContainedGraph {nodes [
   n1 = RigVMFunctionEntryNode()
   SequenceExecution_2 = &RigVMFunction_Sequence(ExecuteContext=@n1.ExecuteContext)
   SequenceExecution_2_1 = &RigVMFunction_Sequence(ExecuteContext=@SequenceExecution_2.B)
   SequenceExecution_2_2 = &RigVMFunction_Sequence(ExecuteContext=@SequenceExecution_2_1.B)
   RigVMFunctionReturnNode(A=@SequenceExecution_2.A, B=@SequenceExecution_2_1.A, C=@SequenceExecution_2_2.A, D=@SequenceExecution_2_2.B)
  ]}
 ContainedGraph {nodes [
   n1 = RigVMFunctionEntryNode()
   SequenceExecution = &RigVMFunction_Sequence(ExecuteContext=@n1.ExecuteContext)
   SequenceExecution_1 = &RigVMFunction_Sequence(ExecuteContext=@SequenceExecution.B)
   SequenceExecution_2 = &RigVMFunction_Sequence(ExecuteContext=@SequenceExecution_1.B)
   RigVMFunctionReturnNode(A=@SequenceExecution.A, B=@SequenceExecution_1.A, C=@SequenceExecution_2.A, D=@SequenceExecution_2.B)
  ]}
 ContainedGraph {nodes [
   n1 = RigVMFunctionEntryNode()
   SequenceExecution_1 = &RigVMFunction_Sequence(ExecuteContext=@n1.ExecuteContext)
   SequenceExecution_1_1 = &RigVMFunction_Sequence(ExecuteContext=@SequenceExecution_1.B)
   SequenceExecution_1_2 = &RigVMFunction_Sequence(ExecuteContext=@SequenceExecution_1_1.B)
   RigVMFunctionReturnNode(A=@SequenceExecution_1.A, B=@SequenceExecution_1_1.A, C=@SequenceExecution_1_2.A, D=@SequenceExecution_1_2.B)
  ]}
 ContainedGraph {nodes [
   n1 = RigVMFunctionEntryNode()
   SequenceExecution_1_1 = &RigVMFunction_Sequence(ExecuteContext=@n1.ExecuteContext)
   SequenceExecution_1_1_1 = &RigVMFunction_Sequence(ExecuteContext=@SequenceExecution_1_1.B)
   SequenceExecution_1_1_2 = &RigVMFunction_Sequence(ExecuteContext=@SequenceExecution_1_1_1.B)
   RigVMFunctionReturnNode(A=@SequenceExecution_1_1.A, B=@SequenceExecution_1_1_1.A, C=@SequenceExecution_1_1_2.A, D=@SequenceExecution_1_1_2.B)
  ]}
 ContainedGraph {nodes [
   n1 = RigVMFunctionEntryNode()
   SequenceExecution_1_1_1 = &RigVMFunction_Sequence(ExecuteContext=@n1.ExecuteContext)
   SequenceExecution_1_1_1_1 = &RigVMFunction_Sequence(ExecuteContext=@SequenceExecution_1_1_1.B)
   SequenceExecution_1_1_1_2 = &RigVMFunction_Sequence(ExecuteContext=@SequenceExecution_1_1_1_1.B)
   RigVMFunctionReturnNode(A=@SequenceExecution_1_1_1.A, B=@SequenceExecution_1_1_1_1.A, C=@SequenceExecution_1_1_1_2.A, D=@SequenceExecution_1_1_1_2.B)
  ]}
 ContainedGraph {nodes [
   n1 = RigVMFunctionEntryNode()
   SequenceExecution_2 = &RigVMFunction_Sequence(ExecuteContext=@n1.ExecuteContext)
   SequenceExecution_2_1 = &RigVMFunction_Sequence(ExecuteContext=@SequenceExecution_2.B)
   SequenceExecution_2_2 = &RigVMFunction_Sequence(ExecuteContext=@SequenceExecution_2_1.B)
   RigVMFunctionReturnNode(A=@SequenceExecution_2.A, B=@SequenceExecution_2_1.A, C=@SequenceExecution_2_2.A, D=@SequenceExecution_2_2.B)
  ]}
 ContainedGraph {nodes [
   n1 = RigVMFunctionEntryNode()
   SequenceExecution_3 = &RigVMFunction_Sequence(ExecuteContext=@n1.ExecuteContext)
   SequenceExecution_3_1 = &RigVMFunction_Sequence(ExecuteContext=@SequenceExecution_3.B)
   SequenceExecution_3_2 = &RigVMFunction_Sequence(ExecuteContext=@SequenceExecution_3_1.B)
   RigVMFunctionReturnNode(A=@SequenceExecution_3.A, B=@SequenceExecution_3_1.A, C=@SequenceExecution_3_2.A, D=@SequenceExecution_3_2.B)
  ]}
 ContainedGraph {nodes [
   n1 = RigVMFunctionEntryNode()
   SequenceExecution_3_1 = &RigVMFunction_Sequence(ExecuteContext=@n1.ExecuteContext)
   SequenceExecution_3_1_1 = &RigVMFunction_Sequence(ExecuteContext=@SequenceExecution_3_1.B)
   SequenceExecution_3_1_2 = &RigVMFunction_Sequence(ExecuteContext=@SequenceExecution_3_1_1.B)
   RigVMFunctionReturnNode(A=@SequenceExecution_3_1.A, B=@SequenceExecution_3_1_1.A, C=@SequenceExecution_3_1_2.A, D=@SequenceExecution_3_1_2.B)
  ]}
 ContainedGraph {nodes [
   n1 = RigVMFunctionEntryNode()
   SequenceExecution_3_1_1 = &RigVMFunction_Sequence(ExecuteContext=@n1.ExecuteContext)
   SequenceExecution_3_1_1_1 = &RigVMFunction_Sequence(ExecuteContext=@SequenceExecution_3_1_1.B)
   SequenceExecution_3_1_1_2 = &RigVMFunction_Sequence(ExecuteContext=@SequenceExecution_3_1_1_1.B)
   RigVMFunctionReturnNode(A=@SequenceExecution_3_1_1.A, B=@SequenceExecution_3_1_1_1.A, C=@SequenceExecution_3_1_1_2.A, D=@SequenceExecution_3_1_1_2.B)
  ]}
 Deform_Elbow_FNC_ContainedGraph {nodes [
   n1 = RigVMFunctionEntryNode()
   > SequenceExecution_2 = RigVMAggregateNode()
   GetDeltaTransformFromRefPose_FNC = RigVMFunctionReferenceNode(Item=(Type=Bone,Name="dfrm_lowerarm_l"), ExecuteContext=@SequenceExecution_2.A)
   GetDeltaTransformFromRefPose_FNC_1 = RigVMFunctionReferenceNode(Item=(Type=Bone,Name="dfrm_lowerarm_r"), ExecuteContext=@SequenceExecution_2.B)
   SequenceExecution = RigVMAggregateNode(ExecuteContext=@SequenceExecution_2.C)
   n2 = &RigVMFunction_ControlFlowBranch(ExecuteContext=@SequenceExecution.A, Condition=@n1.elbow_in)
   SequenceExecution_1 = RigVMAggregateNode(ExecuteContext=@n2.True)
   &RigUnit_ParentConstraint(
    AdvancedSettings=(RotationOrderForFilter=XZY)
    Child=(Type=Bone,Name="deform_elbow_in_base_l")
    Parents=((Item=(Type=Bone,Name="dfrm_lowerarm_l"),Weight=0.4),(Item=(Type=Bone,Name="dfrm_upperarm_l"),Weight=0.6))
    Weight=1
    ExecutePin=@SequenceExecution_1.A
   )
   n3 = &RigVMFunction_MathFloatRemap(SourceMaximum=90, SourceMinimum=15, bClamp=true, Value=@GetDeltaTransformFromRefPose_FNC.OutTransform.Rotation.Yaw)
   > &RigUnit_ModifyTransforms(
    ItemToModify=((Item=(Type=Bone,Name="deform_elbow_in_l"),Transform=(Rotation=(X=0,Y=0,Z=0,W=1),Translation=(X=0,Y=0,Z=0),Scale3D=(X=1,Y=1,Z=1))))
    Weight=@n3.Result
   )
   &RigUnit_ParentConstraint(
    AdvancedSettings=(RotationOrderForFilter=XZY)
    Child=(Type=Bone,Name="deform_elbow_in_base_r")
    Parents=((Item=(Type=Bone,Name="dfrm_lowerarm_r"),Weight=0.4),(Item=(Type=Bone,Name="dfrm_upperarm_r"),Weight=0.6))
    Weight=1
    ExecutePin=@SequenceExecution_1.B
   )
   > Deform_GetMirrorTransform_FNC_1 = RigVMFunctionReferenceNode(transform=(Rotation=(X=0,Y=0,Z=0,W=1),Translation=(X=0,Y=0,Z=0),Scale3D=(X=1,Y=1,Z=1)))
   n4 = &RigVMFunction_MathFloatRemap(SourceMaximum=90, SourceMinimum=15, bClamp=true, Value=@GetDeltaTransformFromRefPose_FNC_1.OutTransform.Rotation.Yaw)
   > &RigUnit_ModifyTransforms(
    ItemToModify=((Item=(Type=Bone,Name="deform_elbow_in_r"),Transform=(Rotation=(X=0,Y=0,Z=0,W=1),Translation=(X=0,Y=0,Z=0),Scale3D=(X=1,Y=1,Z=1))))
    ItemToModify.0.Transform=@Deform_GetMirrorTransform_FNC_1.mirrorTransform
    Weight=@n4.Result
   )
   n5 = &RigVMFunction_ControlFlowBranch(ExecuteContext=@SequenceExecution.C, Condition=@n1.bicep)
   SequenceExecution_1_1_1 = RigVMAggregateNode(ExecuteContext=@n5.True)
   &RigUnit_ModifyTransforms(
    ItemToModify=((Item=(Type=Bone,Name="deform_bicep_l"),Transform=(Rotation=(X=0,Y=0,Z=0,W=1),Translation=(X=0,Y=0,Z=0),Scale3D=(X=1,Y=1,Z=1))))
    ExecutePin=@SequenceExecution_1_1_1.A
    Weight=@n3.Result
   )
   Deform_GetMirrorTransform_FNC = RigVMFunctionReferenceNode(transform=(Rotation=(X=0,Y=0,Z=0,W=1),Translation=(X=0,Y=0,Z=0),Scale3D=(X=1,Y=1,Z=1)), ExecuteContext=@SequenceExecution_1_1_1.B)
   > &RigUnit_ModifyTransforms(
    ItemToModify=((Item=(Type=Bone,Name="deform_bicep_r"),Transform=(Rotation=(X=0,Y=0,Z=0,W=1),Translation=(X=0,Y=0,Z=0),Scale3D=(X=1,Y=1,Z=1))))
    ItemToModify.0.Transform=@Deform_GetMirrorTransform_FNC.mirrorTransform
    Weight=@n4.Result
   )
   n6 = &RigVMFunction_ControlFlowBranch(Condition=@n1.elbow_out, ExecuteContext=@SequenceExecution.B)
   SequenceExecution_1_1 = RigVMAggregateNode(ExecuteContext=@n6.True)
   &RigUnit_ParentConstraint(
    AdvancedSettings=(RotationOrderForFilter=XZY)
    Child=(Type=Bone,Name="deform_elbow_out_base_l")
    Parents=((Item=(Type=Bone,Name="dfrm_lowerarm_l"),Weight=0.3),(Item=(Type=Bone,Name="dfrm_upperarm_l"),Weight=0.7))
    Weight=1
    ExecutePin=@SequenceExecution_1_1.A
   )
   > &RigUnit_ModifyTransforms(
    ItemToModify=((Item=(Type=Bone,Name="deform_elbow_out_l"),Transform=(Rotation=(X=0,Y=0,Z=0,W=1),Translation=(X=0,Y=0,Z=0),Scale3D=(X=1,Y=1,Z=1))))
    Weight=@n3.Result
   )
   &RigUnit_ParentConstraint(
    AdvancedSettings=(RotationOrderForFilter=XZY)
    Child=(Type=Bone,Name="deform_elbow_out_base_r")
    Parents=((Item=(Type=Bone,Name="dfrm_lowerarm_r"),Weight=0.3),(Item=(Type=Bone,Name="dfrm_upperarm_r"),Weight=0.7))
    Weight=1
    ExecutePin=@SequenceExecution_1_1.B
   )
   > Deform_GetMirrorTransform_FNC_2 = RigVMFunctionReferenceNode(transform=(Rotation=(X=0,Y=0,Z=0,W=1),Translation=(X=0,Y=0,Z=0),Scale3D=(X=1,Y=1,Z=1)))
   > &RigUnit_ModifyTransforms(
    ItemToModify=((Item=(Type=Bone,Name="deform_elbow_out_r"),Transform=(Rotation=(X=0,Y=0,Z=0,W=1),Translation=(X=0,Y=0,Z=0),Scale3D=(X=1,Y=1,Z=1))))
    ItemToModify.0.Transform=@Deform_GetMirrorTransform_FNC_2.mirrorTransform
    Weight=@n4.Result
   )
   RigVMFunctionReturnNode(ExecuteContext=@SequenceExecution_2.D)
  ]}
 Deform_GetMirrorTransform_FNC_2_ContainedGraph {nodes [
   n1 = RigVMFunctionEntryNode()
   n2 = &RigVMFunction_MathTransformToEulerTransform(Value=@n1.transform) {template="Cast::Execute(in Value,out Result)"}
   > RigVMFunctionReturnNode(
    mirrorTransform=(Rotation=(X=0,Y=0,Z=0,W=1),Translation=(X=0,Y=0,Z=0),Scale3D=(X=1,Y=1,Z=1))
    mirrorTransform.Rotation=&RigVMFunction_MathQuaternionFromRotatorV2(Value=@n2.Result.Rotation) {template="Cast::Execute(in Value,out Result)"}.Result
    mirrorTransform.Translation=&RigVMFunction_MathVectorScale(Factor=-1, Value=@n2.Result.Location) {template="Scale::Execute(in Value,in Factor,out Result)"}.Result
    mirrorTransform.Scale3D=@n2.Result.Scale
   )
  ]}
 Deform_GetMirrorTransform_FNC_3_ContainedGraph {nodes [
   n1 = RigVMFunctionEntryNode()
   n2 = &RigVMFunction_MathTransformToEulerTransform(Value=@n1.transform) {template="Cast::Execute(in Value,out Result)"}
   > RigVMFunctionReturnNode(
    mirrorTransform=(Rotation=(X=0,Y=0,Z=0,W=1),Translation=(X=0,Y=0,Z=0),Scale3D=(X=1,Y=1,Z=1))
    mirrorTransform.Rotation=&RigVMFunction_MathQuaternionFromRotatorV2(Value=@n2.Result.Rotation) {template="Cast::Execute(in Value,out Result)"}.Result
    mirrorTransform.Translation=&RigVMFunction_MathVectorScale(Factor=-1, Value=@n2.Result.Location) {template="Scale::Execute(in Value,in Factor,out Result)"}.Result
    mirrorTransform.Scale3D=@n2.Result.Scale
   )
  ]}
 Deform_GetMirrorTransform_FNC_4_ContainedGraph {nodes [
   n1 = RigVMFunctionEntryNode()
   n2 = &RigVMFunction_MathTransformToEulerTransform(Value=@n1.transform) {template="Cast::Execute(in Value,out Result)"}
   > RigVMFunctionReturnNode(
    mirrorTransform=(Rotation=(X=0,Y=0,Z=0,W=1),Translation=(X=0,Y=0,Z=0),Scale3D=(X=1,Y=1,Z=1))
    mirrorTransform.Rotation=&RigVMFunction_MathQuaternionFromRotatorV2(Value=@n2.Result.Rotation) {template="Cast::Execute(in Value,out Result)"}.Result
    mirrorTransform.Translation=&RigVMFunction_MathVectorScale(Factor=-1, Value=@n2.Result.Location) {template="Scale::Execute(in Value,in Factor,out Result)"}.Result
    mirrorTransform.Scale3D=@n2.Result.Scale
   )
  ]}
 Deform_UpperArm_FNC_ContainedGraph {nodes [
   n1 = RigVMFunctionEntryNode()
   > SequenceExecution_2 = RigVMAggregateNode()
   n2 = &RigVMFunction_ControlFlowBranch(Condition=@n1.upperarm_up, ExecuteContext=@SequenceExecution_2.A)
   SequenceExecution_3 = RigVMAggregateNode(ExecuteContext=@n2.True)
   &RigUnit_ParentConstraint(
    AdvancedSettings=(RotationOrderForFilter=XZY)
    Child=(Type=Bone,Name="deform_upperarm_up_base_l")
    Parents=((Item=(Type=Bone,Name="dfrm_upperarm_twist_01_l"),Weight=0.5),(Item=(Type=Bone,Name="clavicle_l"),Weight=0.5))
    Weight=1
    ExecutePin=@SequenceExecution_3.A
   )
   > n3 = &RigUnit_SphericalPoseReader(
    ActiveRegionScaleFactors=(PositiveWidth=2.5,NegativeWidth=3,PositiveHeight=3,NegativeHeight=3)
    ActiveRegionSize=0.1
    Debug=(bDrawDebug=false,DebugScale=25,DebugSegments=20,DebugThickness=0.25)
    DriverAxis=(X=1,Y=0,Z=0)
    DriverItem=(Type=Bone,Name="dfrm_upperarm_l")
    FalloffRegionScaleFactors=(PositiveWidth=1,NegativeWidth=1,PositiveHeight=1,NegativeHeight=1)
    FalloffSize=0.5
    OptionalParentItem=(Type=Bone)
    RotationOffset=(X=0,Y=0,Z=0)
    Debug.bDrawDebug=@n1.debug_up
   )
   > &RigUnit_ModifyTransforms(
    ItemToModify=((Item=(Type=Bone,Name="deform_upperarm_up_l"),Transform=(Rotation=(X=0,Y=0,Z=0,W=1),Translation=(X=0,Y=0,Z=5),Scale3D=(X=1,Y=1,Z=1))))
    Weight=@n3.OutputParam
   )
   &RigUnit_ParentConstraint(
    AdvancedSettings=(RotationOrderForFilter=XZY)
    Child=(Type=Bone,Name="deform_upperarm_up_base_r")
    Parents=((Item=(Type=Bone,Name="dfrm_upperarm_twist_01_r"),Weight=0.5),(Item=(Type=Bone,Name="clavicle_r"),Weight=0.5))
    Weight=1
    ExecutePin=@SequenceExecution_3.B
   )
   > n4 = &RigUnit_SphericalPoseReader(
    ActiveRegionScaleFactors=(PositiveWidth=2.5,NegativeWidth=3,PositiveHeight=3,NegativeHeight=3)
    ActiveRegionSize=0.1
    Debug=(bDrawDebug=false,DebugScale=25,DebugSegments=20,DebugThickness=0.25)
    DriverAxis=(X=-1,Y=0,Z=0)
    DriverItem=(Type=Bone,Name="dfrm_upperarm_r")
    FalloffRegionScaleFactors=(PositiveWidth=1,NegativeWidth=1,PositiveHeight=1,NegativeHeight=1)
    FalloffSize=0.5
    FlipHeightScaling=true
    OptionalParentItem=(Type=Bone)
    RotationOffset=(X=180,Y=0,Z=180)
    Debug.bDrawDebug=@n1.debug_up
   )
   > Deform_GetMirrorTransform_FNC = RigVMFunctionReferenceNode(transform=(Rotation=(X=0,Y=0,Z=0,W=1),Translation=(X=0,Y=0,Z=0),Scale3D=(X=1,Y=1,Z=1)))
   > &RigUnit_ModifyTransforms(
    ItemToModify=((Item=(Type=Bone,Name="deform_upperarm_up_r"),Transform=(Rotation=(X=0,Y=0,Z=0,W=1),Translation=(X=0,Y=0,Z=0),Scale3D=(X=1,Y=1,Z=1))))
    ItemToModify.0.Transform=@Deform_GetMirrorTransform_FNC.mirrorTransform
    Weight=@n4.OutputParam
   )
   n5 = &RigVMFunction_ControlFlowBranch(Condition=@n1.upperarm_fr, ExecuteContext=@SequenceExecution_2.B)
   SequenceExecution_3_1 = RigVMAggregateNode(ExecuteContext=@n5.True)
   &RigUnit_ParentConstraint(
    AdvancedSettings=(RotationOrderForFilter=XZY)
    Child=(Type=Bone,Name="deform_upperarm_fr_base_l")
    Parents=((Item=(Type=Bone,Name="dfrm_upperarm_twist_01_l"),Weight=0.3),(Item=(Type=Bone,Name="clavicle_l"),Weight=0.7))
    Weight=1
    ExecutePin=@SequenceExecution_3_1.A
   )
   > n6 = &RigUnit_SphericalPoseReader(
    ActiveRegionScaleFactors=(PositiveWidth=0.5,NegativeWidth=3,PositiveHeight=1.5,NegativeHeight=1.5)
    ActiveRegionSize=0.1
    Debug=(bDrawDebug=false,DebugScale=25,DebugSegments=20,DebugThickness=0.25)
    DriverAxis=(X=1,Y=0,Z=0)
    DriverItem=(Type=Bone,Name="dfrm_upperarm_l")
    FalloffRegionScaleFactors=(PositiveWidth=1,NegativeWidth=1,PositiveHeight=1,NegativeHeight=1)
    FalloffSize=0.4
    OptionalParentItem=(Type=Bone)
    RotationOffset=(X=90,Y=0,Z=0)
    Debug.bDrawDebug=@n1.debug_fr
   )
   > &RigUnit_ModifyTransforms(
    ItemToModify=((Item=(Type=Bone,Name="deform_upperarm_fr_l"),Transform=(Rotation=(X=0,Y=0,Z=0,W=1),Translation=(X=0,Y=0,Z=0),Scale3D=(X=1,Y=1,Z=1))))
    Weight=@n6.OutputParam
   )
   > &RigUnit_ModifyTransforms(
    ItemToModify=((Item=(Type=Bone,Name="deform_upperarm_bk_l"),Transform=(Rotation=(X=0,Y=0,Z=0,W=1),Translation=(X=0,Y=0,Z=0),Scale3D=(X=1,Y=1,Z=1))))
    Weight=@n6.OutputParam
   )
   &RigUnit_ParentConstraint(
    AdvancedSettings=(RotationOrderForFilter=XZY)
    Child=(Type=Bone,Name="deform_upperarm_fr_base_r")
    Parents=((Item=(Type=Bone,Name="dfrm_upperarm_twist_01_r"),Weight=0.3),(Item=(Type=Bone,Name="clavicle_r"),Weight=0.7))
    Weight=1
    ExecutePin=@SequenceExecution_3_1.B
   )
   > n7 = &RigUnit_SphericalPoseReader(
    ActiveRegionScaleFactors=(PositiveWidth=0.5,NegativeWidth=3,PositiveHeight=1.5,NegativeHeight=1.5)
    ActiveRegionSize=0.1
    Debug=(bDrawDebug=false,DebugScale=25,DebugSegments=20,DebugThickness=0.25)
    DriverAxis=(X=-1,Y=0,Z=0)
    DriverItem=(Type=Bone,Name="dfrm_upperarm_r")
    FalloffRegionScaleFactors=(PositiveWidth=1,NegativeWidth=1,PositiveHeight=1,NegativeHeight=1)
    FalloffSize=0.4
    FlipWidthScaling=true
    OptionalParentItem=(Type=Bone)
    RotationOffset=(X=90,Y=180,Z=180)
    Debug.bDrawDebug=@n1.debug_fr
   )
   > Deform_GetMirrorTransform_FNC_2 = RigVMFunctionReferenceNode(transform=(Rotation=(X=0,Y=0,Z=0,W=1),Translation=(X=0,Y=0,Z=0),Scale3D=(X=1,Y=1,Z=1)))
   > &RigUnit_ModifyTransforms(
    ItemToModify=((Item=(Type=Bone,Name="deform_upperarm_fr_r"),Transform=(Rotation=(X=0,Y=0,Z=0,W=1),Translation=(X=0,Y=0,Z=0),Scale3D=(X=1,Y=1,Z=1))))
    ItemToModify.0.Transform=@Deform_GetMirrorTransform_FNC_2.mirrorTransform
    Weight=@n7.OutputParam
   )
   > Deform_GetMirrorTransform_FNC_1 = RigVMFunctionReferenceNode(transform=(Rotation=(X=0,Y=0,Z=0,W=1),Translation=(X=0,Y=0,Z=0),Scale3D=(X=1,Y=1,Z=1)))
   > &RigUnit_ModifyTransforms(
    ItemToModify=((Item=(Type=Bone,Name="deform_upperarm_bk_r"),Transform=(Rotation=(X=0,Y=0,Z=0,W=1),Translation=(X=0,Y=0,Z=0),Scale3D=(X=1,Y=1,Z=1))))
    ItemToModify.0.Transform=@Deform_GetMirrorTransform_FNC_1.mirrorTransform
    Weight=@n7.OutputParam
   )
   n8 = &RigVMFunction_ControlFlowBranch(Condition=@n1.upperarm_bk, ExecuteContext=@SequenceExecution_2.C)
   SequenceExecution_3_1_1 = RigVMAggregateNode(ExecuteContext=@n8.True)
   &RigUnit_ParentConstraint(
    AdvancedSettings=(RotationOrderForFilter=XZY)
    Child=(Type=Bone,Name="deform_upperarm_bk_base_l")
    Parents=((Item=(Type=Bone,Name="dfrm_upperarm_twist_01_l"),Weight=0.3),(Item=(Type=Bone,Name="clavicle_l"),Weight=0.7))
    Weight=1
    ExecutePin=@SequenceExecution_3_1_1.A
   )
   > n9 = &RigUnit_SphericalPoseReader(
    ActiveRegionScaleFactors=(PositiveWidth=3,NegativeWidth=1,PositiveHeight=1.5,NegativeHeight=1.5)
    ActiveRegionSize=0.1
    Debug=(bDrawDebug=false,DebugScale=25,DebugSegments=20,DebugThickness=0.25)
    DriverAxis=(X=1,Y=0,Z=0)
    DriverItem=(Type=Bone,Name="dfrm_upperarm_l")
    FalloffRegionScaleFactors=(PositiveWidth=1,NegativeWidth=1,PositiveHeight=1,NegativeHeight=1)
    FalloffSize=0.4
    OptionalParentItem=(Type=Bone)
    RotationOffset=(X=90,Y=0,Z=180)
    Debug.bDrawDebug=@n1.debug_bk
   )
   > &RigUnit_ModifyTransforms(
    ItemToModify=((Item=(Type=Bone,Name="deform_upperarm_bk_l"),Transform=(Rotation=(X=0,Y=0,Z=0,W=1),Translation=(X=0,Y=0,Z=0),Scale3D=(X=1,Y=1,Z=1))))
    Weight=@n9.OutputParam
   )
   &RigUnit_ParentConstraint(
    AdvancedSettings=(RotationOrderForFilter=XZY)
    Child=(Type=Bone,Name="deform_upperarm_bk_base_r")
    Parents=((Item=(Type=Bone,Name="dfrm_upperarm_twist_01_r"),Weight=0.3),(Item=(Type=Bone,Name="clavicle_r"),Weight=0.7))
    Weight=1
    ExecutePin=@SequenceExecution_3_1_1.B
   )
   > n10 = &RigUnit_SphericalPoseReader(
    ActiveRegionScaleFactors=(PositiveWidth=3,NegativeWidth=1,PositiveHeight=1.5,NegativeHeight=1.5)
    ActiveRegionSize=0.1
    Debug=(bDrawDebug=false,DebugScale=25,DebugSegments=20,DebugThickness=0.25)
    DriverAxis=(X=-1,Y=0,Z=0)
    DriverItem=(Type=Bone,Name="dfrm_upperarm_r")
    FalloffRegionScaleFactors=(PositiveWidth=1,NegativeWidth=1,PositiveHeight=1,NegativeHeight=1)
    FalloffSize=0.4
    FlipHeightScaling=true
    FlipWidthScaling=false
    OptionalParentItem=(Type=Bone)
    RotationOffset=(X=90,Y=0,Z=0)
    Debug.bDrawDebug=@n1.debug_bk
   )
   > Deform_GetMirrorTransform_FNC_3 = RigVMFunctionReferenceNode(transform=(Rotation=(X=0,Y=0,Z=0,W=1),Translation=(X=0,Y=0,Z=0),Scale3D=(X=1,Y=1,Z=1)))
   > &RigUnit_ModifyTransforms(
    ItemToModify=((Item=(Type=Bone,Name="deform_upperarm_bk_r"),Transform=(Rotation=(X=0,Y=0,Z=0,W=1),Translation=(X=0,Y=0,Z=0),Scale3D=(X=1,Y=1,Z=1))))
    ItemToModify.0.Transform=@Deform_GetMirrorTransform_FNC_3.mirrorTransform
    Weight=@n10.OutputParam
   )
   RigVMFunctionReturnNode(ExecuteContext=@SequenceExecution_2.D)
  ]}
 GetDeltaTransformFromRefPose_FNC_1_ContainedGraph {nodes [
   n1 = RigVMFunctionEntryNode()
   n2 = &RigUnit_GetTransform(Space=LocalSpace, Item=@n1.Item)
   n3 = &RigUnit_GetTransform(Space=LocalSpace, bInitial=true, Item=@n1.Item)
   > RigVMFunctionReturnNode(OutTransform=&RigVMFunction_MathTransformToEulerTransform(Value=&RigVMFunction_MathTransformMakeRelative(Global=@n2.Transform, Parent=@n3.Transform).Local) {template="Cast::Execute(in Value,out Result)"}.Result)
  ]}
 RigVMFunctionLibrary {nodes [
   Deform_Elbow_FNC = RigVMCollapseNode(ExecuteContext=(), bicep=true, elbow_in=true, elbow_out=true)
   Deform_GetMirrorTransform_FNC = RigVMCollapseNode(ExecuteContext=(), transform=(Rotation=(X=0,Y=0,Z=0,W=1),Translation=(X=0,Y=0,Z=0),Scale3D=(X=1,Y=1,Z=1)))
   Deform_GetMirrorTransform_FNC_1 = RigVMCollapseNode(ExecuteContext=(), transform=(Rotation=(X=0,Y=0,Z=0,W=1),Translation=(X=0,Y=0,Z=0),Scale3D=(X=1,Y=1,Z=1)))
   Deform_GetMirrorTransform_FNC_2 = RigVMCollapseNode(ExecuteContext=(), transform=(Rotation=(X=0,Y=0,Z=0,W=1),Translation=(X=0,Y=0,Z=0),Scale3D=(X=1,Y=1,Z=1)))
   Deform_GetMirrorTransform_FNC_3 = RigVMCollapseNode(ExecuteContext=(), transform=(Rotation=(X=0,Y=0,Z=0,W=1),Translation=(X=0,Y=0,Z=0),Scale3D=(X=1,Y=1,Z=1)))
   Deform_GetMirrorTransform_FNC_4 = RigVMCollapseNode(ExecuteContext=(), transform=(Rotation=(X=0,Y=0,Z=0,W=1),Translation=(X=0,Y=0,Z=0),Scale3D=(X=1,Y=1,Z=1)))
   Deform_GroinGlue_FNC = RigVMCollapseNode(ExecuteContext=(), debug=False, glute=true, groin=true)
   Deform_Knee_FNC = RigVMCollapseNode(ExecuteContext=(), knee_in=true, knee_out=true)
   Deform_UpperArm_FNC = RigVMCollapseNode(ExecuteContext=(), debug_bk=false, debug_fr=false, debug_up=false, upperarm_bk=true, upperarm_fr=true, upperarm_up=true)
   GetDeltaTransformFromRefPose_FNC = RigVMCollapseNode(ExecuteContext=(), Item=(Type=Bone))
   GetDeltaTransformFromRefPose_FNC_1 = RigVMCollapseNode(ExecuteContext=(), Item=(Type=None,Name="None"))
  ]}
 RigVMModel {nodes [
   RigUnit_BeginExecution = &RigUnit_BeginExecution()
   SequenceAggregate_1 = RigVMAggregateNode(ExecuteContext=@RigUnit_BeginExecution.ExecutePin)
   &RigUnit_ParentConstraint(
    AdvancedSettings=(InterpolationType=Average,RotationOrderForFilter=XZY)
    Child=(Type=Bone,Name="dfrm_thigh_l")
    Filter=(TranslationFilter=(bX=false,bY=false,bZ=false),RotationFilter=(bX=True,bY=True,bZ=True),ScaleFilter=(bX=True,bY=True,bZ=True))
    Parents=((Item=(Type=Bone,Name="thigh_l"),Weight=1))
    Weight=1
    ExecutePin=@SequenceAggregate_1.A
   )
   > &RigUnit_ParentConstraint(
    AdvancedSettings=(InterpolationType=Average,RotationOrderForFilter=XZY)
    Child=(Type=Bone,Name="dfrm_thigh_r")
    Filter=(TranslationFilter=(bX=false,bY=false,bZ=false),RotationFilter=(bX=True,bY=True,bZ=True),ScaleFilter=(bX=True,bY=True,bZ=True))
    Parents=((Item=(Type=Bone,Name="thigh_r"),Weight=1))
    Weight=1
   )
   > n1 = &RigUnit_SphericalPoseReader(
    ActiveRegionScaleFactors=(PositiveWidth=3.5,NegativeWidth=0.5,PositiveHeight=3,NegativeHeight=3)
    ActiveRegionSize=0.1
    Debug=(bDrawDebug=false,DebugScale=25,DebugSegments=20,DebugThickness=0.25)
    DriverAxis=(X=1,Y=0,Z=0)
    DriverItem=(Type=Bone,Name="dfrm_thigh_r")
    FalloffRegionScaleFactors=(PositiveWidth=1,NegativeWidth=1,PositiveHeight=1,NegativeHeight=1)
    FalloffSize=0.4
    FlipHeightScaling=false
    FlipWidthScaling=true
    OptionalParentItem=(Type=Bone)
    RotationOffset=(X=-90,Y=0,Z=0)
   )
   > RigVMVariableNode(Value=@n1.OutputParam) {variable=thigh r up}
   n2 = &RigVMFunction_AlphaInterp(
    InterpSpeedDecreasing=10
    InterpSpeedIncreasing=10
    Scale=1
    bClampResult=true
    bMapRange=true
    Value=&RigVMFunction_MathDoubleRemap(SourceMaximum=0, SourceMinimum=1, TargetMaximum=1, TargetMinimum=0, Value=RigVMVariableNode() {variable=thigh r up}.Value).Result
   )
   > &RigUnit_ModifyTransforms(
    ItemToModify=((Item=(Type=Bone,Name="dfrm_calf_r_PV"),Transform=(Rotation=(X=0,Y=0,Z=0,W=1),Translation=(X=0,Y=0,Z=45),Scale3D=(X=1,Y=1,Z=1))))
    WeightMinimum=0
    Weight=@n2.Result
   )
   > &RigUnit_ParentConstraint(
    AdvancedSettings=(InterpolationType=Average,RotationOrderForFilter=XZY)
    Child=(Type=Control,Name="PV L")
    Filter=(TranslationFilter=(bX=true,bY=true,bZ=true),RotationFilter=(bX=True,bY=True,bZ=True),ScaleFilter=(bX=True,bY=True,bZ=True))
    Parents=((Item=(Type=Bone,Name="dfrm_calf_l_PV"),Weight=1))
    Weight=1
   )
   > &RigUnit_ParentConstraint(
    AdvancedSettings=(InterpolationType=Average,RotationOrderForFilter=XZY)
    Child=(Type=Control,Name="PV R")
    Filter=(TranslationFilter=(bX=true,bY=true,bZ=true),RotationFilter=(bX=True,bY=True,bZ=True),ScaleFilter=(bX=True,bY=True,bZ=True))
    Parents=((Item=(Type=Bone,Name="dfrm_calf_r_PV"),Weight=1))
    Weight=1
   )
   > /Script/ControlRig.RigUnit_TwoBoneIKSimplePerItem(
    DebugSettings=(bEnabled=False,Scale=10,WorldOffset=(Rotation=(X=0,Y=0,Z=0,W=1),Translation=(X=0,Y=0,Z=0),Scale3D=(X=1,Y=1,Z=1)))
    EffectorItem=(Type=Bone,Name="dfrm_foot_l")
    ItemA=(Type=Bone,Name="dfrm_thigh_l")
    ItemALength=0
    ItemB=(Type=Bone,Name="dfrm_calf_l")
    ItemBLength=0
    PoleVectorKind=Direction
    PoleVectorSpace=(Type=Bone,Name="None")
    PrimaryAxis=(X=-1,Y=0,Z=0)
    SecondaryAxis=(X=0,Y=1,Z=0)
    SecondaryAxisWeight=1
    StretchMaximumRatio=1.25
    StretchStartRatio=0.99
    Weight=1
    bEnableStretch=true
    bPropagateToChildren=True
    PoleVector=&RigUnit_GetTransform(Item=(Type=Control,Name="PV L"), bInitial=false).Transform.Translation
    Effector=&RigUnit_GetTransform(Item=(Type=Control,Name="dfrm_foot_l_ctrl"), bInitial=false).Transform
   ) {is=struct RigVMUnitNode}
   > &RigUnit_TwoBoneIKSimplePerItem(
    DebugSettings=(Scale=10,WorldOffset=(Rotation=(X=0,Y=0,Z=0,W=1),Translation=(X=0,Y=0,Z=0),Scale3D=(X=1,Y=1,Z=1)))
    EffectorItem=(Type=Bone,Name="dfrm_foot_r")
    ItemA=(Type=Bone,Name="dfrm_thigh_r")
    ItemB=(Type=Bone,Name="dfrm_calf_r")
    PoleVectorSpace=(Type=Bone,Name="None")
    PrimaryAxis=(X=1,Y=0,Z=0)
    SecondaryAxis=(X=0,Y=-1,Z=0)
    SecondaryAxisWeight=1
    StretchMaximumRatio=1.25
    StretchStartRatio=0.99
    Weight=1
    bEnableStretch=true
    bPropagateToChildren=True
    Effector=&RigUnit_GetTransform(Item=(Type=Control,Name="dfrm_foot_r_ctrl")).Transform
    PoleVector=&RigUnit_GetTransform(Item=(Type=Control,Name="PV R")).Transform.Translation
   )
   RigVMFunction_MathBoolToFloat = &RigVMFunction_MathBoolToFloat(Value=RigVMVariableNode() {variable=isFrontend}.Value)
   > &RigUnit_ParentConstraint(
    AdvancedSettings=(InterpolationType=Average,RotationOrderForFilter=XZY)
    Child=(Type=Bone,Name="dfrm_foot_l")
    Filter=(TranslationFilter=(bX=true,bY=true,bZ=true),RotationFilter=(bX=True,bY=True,bZ=True),ScaleFilter=(bX=True,bY=True,bZ=True))
    Parents=((Item=(Type=Bone,Name="foot_l"),Weight=1))
    Weight=@RigVMFunction_MathBoolToFloat.Result
   )
   > &RigUnit_ParentConstraint(
    AdvancedSettings=(InterpolationType=Average,RotationOrderForFilter=XZY)
    Child=(Type=Bone,Name="dfrm_foot_r")
    Filter=(TranslationFilter=(bX=true,bY=true,bZ=true),RotationFilter=(bX=True,bY=True,bZ=True),ScaleFilter=(bX=True,bY=True,bZ=True))
    Parents=((Item=(Type=Bone,Name="foot_r"),Weight=1))
    Weight=@RigVMFunction_MathBoolToFloat.Result
   )
   n3 = &RigVMFunction_MathDoubleMul(
    A=&RigVMFunction_MathBoolToFloat(Value=RigVMVariableNode() {variable=isFrontend}.Value).Result
    B=&RigVMFunction_MathBoolToFloat(Value=RigVMVariableNode() {variable=isEmote}.Value).Result
   )
   > &RigUnit_ModifyTransforms(
    ItemToModify=((Item=(Type=Bone,Name="dfrm_foot_l"),Transform=(Rotation=(X=0,Y=0,Z=0,W=1),Translation=(X=0,Y=-1,Z=0),Scale3D=(X=1,Y=1,Z=1))))
    WeightMinimum=0
    Weight=@n3.Result
   )
   > &RigUnit_ModifyTransforms(
    ItemToModify=((Item=(Type=Bone,Name="dfrm_foot_r"),Transform=(Rotation=(X=0,Y=0,Z=0,W=1),Translation=(X=0,Y=1,Z=0),Scale3D=(X=1,Y=1,Z=1))))
    WeightMinimum=0
    Weight=@n3.Result
   )
   > &RigUnit_ParentConstraint(
    AdvancedSettings=(InterpolationType=Average,RotationOrderForFilter=XZY)
    Child=(Type=Bone,Name="dfrm_thigh_twist_01_l")
    Filter=(TranslationFilter=(bX=false,bY=false,bZ=false),RotationFilter=(bX=True,bY=True,bZ=True),ScaleFilter=(bX=True,bY=True,bZ=True))
    Parents=((Item=(Type=Bone,Name="thigh_twist_01_l"),Weight=1))
    Weight=1
   )
   > &RigUnit_ParentConstraint(
    AdvancedSettings=(InterpolationType=Average,RotationOrderForFilter=XZY)
    Child=(Type=Bone,Name="dfrm_thigh_twist_01_r")
    Filter=(TranslationFilter=(bX=false,bY=false,bZ=false),RotationFilter=(bX=True,bY=True,bZ=True),ScaleFilter=(bX=True,bY=True,bZ=True))
    Parents=((Item=(Type=Bone,Name="thigh_twist_01_r"),Weight=1))
    Weight=1
   )
   > &RigUnit_ParentConstraint(
    AdvancedSettings=(InterpolationType=Average,RotationOrderForFilter=XZY)
    Child=(Type=Bone,Name="dfrm_calf_twist_01_l")
    Filter=(TranslationFilter=(bX=false,bY=true,bZ=true),RotationFilter=(bX=True,bY=True,bZ=True),ScaleFilter=(bX=True,bY=True,bZ=True))
    Parents=((Item=(Type=Bone,Name="calf_twist_01_l"),Weight=1))
    Weight=1
   )
   > &RigUnit_ParentConstraint(
    AdvancedSettings=(InterpolationType=Average,RotationOrderForFilter=XZY)
    Child=(Type=Bone,Name="dfrm_calf_twist_01_r")
    Filter=(TranslationFilter=(bX=false,bY=true,bZ=true),RotationFilter=(bX=True,bY=True,bZ=True),ScaleFilter=(bX=True,bY=True,bZ=True))
    Parents=((Item=(Type=Bone,Name="calf_twist_01_r"),Weight=1))
    Weight=1
   )
   > &RigUnit_ParentConstraint(
    AdvancedSettings=(InterpolationType=Average,RotationOrderForFilter=XZY)
    Child=(Type=Bone,Name="dfrm_calf_twist_02_l")
    Filter=(TranslationFilter=(bX=false,bY=false,bZ=true),RotationFilter=(bX=true,bY=true,bZ=true),ScaleFilter=(bX=True,bY=True,bZ=True))
    Parents=((Item=(Type=Bone,Name="calf_twist_02_l"),Weight=1))
    Weight=1
   )
   > &RigUnit_ParentConstraint(
    AdvancedSettings=(InterpolationType=Average,RotationOrderForFilter=XZY)
    Child=(Type=Bone,Name="dfrm_calf_twist_02_r")
    Filter=(TranslationFilter=(bX=false,bY=false,bZ=true),RotationFilter=(bX=True,bY=True,bZ=True),ScaleFilter=(bX=True,bY=True,bZ=True))
    Parents=((Item=(Type=Bone,Name="calf_twist_02_r"),Weight=1))
    Weight=1
   )
   > &RigUnit_RotationConstraintLocalSpaceOffset(
    AdvancedSettings=(InterpolationType=Average,RotationOrderForFilter=XZY)
    Child=(Type=Bone,Name="dyn_ball_l")
    Filter=(bX=True,bY=True,bZ=True)
    Parents=((Item=(Type=Bone,Name="ball_l"),Weight=1))
    Weight=1
   )
   > &RigUnit_RotationConstraintLocalSpaceOffset(
    AdvancedSettings=(InterpolationType=Average,RotationOrderForFilter=XZY)
    Child=(Type=Bone,Name="dyn_ball_r")
    Filter=(bX=True,bY=True,bZ=True)
    Parents=((Item=(Type=Bone,Name="ball_r"),Weight=1))
    Weight=1
   )
   SequenceAggregate_2_1 = RigVMAggregateNode(ExecuteContext=@SequenceAggregate_1.B)
   n4 = &RigVMFunction_ControlFlowBranch(Condition=RigVMVariableNode() {variable=deform_bool}.Value.knee_9_#1, ExecuteContext=@SequenceAggregate_2_1.A)
   Deform_Knee_FNC_1 = RigVMFunctionReferenceNode(knee_in=true, knee_out=true, ExecuteContext=@n4.True)
   n5 = RigVMVariableNode() {variable=deform_bool}
   n6 = &RigVMFunction_ControlFlowBranch(
    Condition=&RigVMFunction_MathBoolOr(B=@n5.Value.glute_17_#2, A=@n5.Value.groin_8_#3).Result
    ExecuteContext=@SequenceAggregate_2_1.B
   )
   Deform_GroinGlue_FNC_1 = RigVMFunctionReferenceNode(
    debug=False
    ExecuteContext=@n6.True
    glute=@n5.Value.glute_17_#2
    groin=@n5.Value.groin_8_#3
   )
   n7 = &RigUnit_SphericalPoseReader(
    ActiveRegionScaleFactors=(PositiveWidth=1,NegativeWidth=1,PositiveHeight=1,NegativeHeight=1)
    ActiveRegionSize=0.2
    Debug=(bDrawDebug=false,bDraw2D=False,bDrawLocalAxes=False,DebugScale=25,DebugSegments=20,DebugThickness=0.25)
    DriverAxis=(X=1,Y=0,Z=0)
    DriverItem=(Type=Bone,Name="upperarm_l")
    FalloffRegionScaleFactors=(PositiveWidth=1,NegativeWidth=1,PositiveHeight=1,NegativeHeight=1)
    FalloffSize=0.3
    FlipHeightScaling=False
    FlipWidthScaling=False
    OptionalParentItem=(Type=Bone,Name="None")
    RotationOffset=(X=20,Y=-10,Z=0)
    ExecutePin=@SequenceAggregate_1.C
   )
   > &RigUnit_ModifyTransforms(
    ItemToModify=((Item=(Type=Bone,Name="dyn_collar_l_MSTR"),Transform=(Rotation=(X=0,Y=-0.2164,Z=0,W=0.9763),Translation=(X=-2.5181,Y=-1.5826,Z=2.3601),Scale3D=(X=1,Y=1,Z=1))))
    WeightMinimum=0
    Weight=@n7.OutputParam
   )
   > n8 = &RigUnit_SphericalPoseReader(
    ActiveRegionScaleFactors=(PositiveWidth=1.2,NegativeWidth=1,PositiveHeight=0.5,NegativeHeight=1)
    ActiveRegionSize=0.1
    Debug=(bDrawDebug=false,bDraw2D=False,bDrawLocalAxes=False,DebugScale=25,DebugSegments=20,DebugThickness=0.25)
    DriverAxis=(X=1,Y=0,Z=0)
    DriverItem=(Type=Bone,Name="upperarm_l")
    FalloffRegionScaleFactors=(PositiveWidth=1,NegativeWidth=1,PositiveHeight=1,NegativeHeight=1)
    FalloffSize=0.25
    FlipHeightScaling=False
    FlipWidthScaling=False
    OptionalParentItem=(Type=Bone,Name="None")
    RotationOffset=(X=65,Y=0,Z=-20)
   )
   > &RigUnit_ModifyTransforms(
    ItemToModify=((Item=(Type=Bone,Name="dyn_collar_l_1"),Transform=(Rotation=(X=0,Y=0.1736,Z=0,W=-0.9848),Translation=(X=-0.8,Y=-2,Z=0),Scale3D=(X=1,Y=1,Z=1))))
    WeightMinimum=0
    Weight=@n8.OutputParam
   )
   > &RigUnit_ModifyTransforms(
    ItemToModify=((Item=(Type=Bone,Name="dyn_collar_l_2"),Transform=(Rotation=(X=-0.0817,Y=-0.3487,Z=-0.3372,W=0.8707),Translation=(X=-0.8547,Y=1.0748,Z=2.9823),Scale3D=(X=1,Y=1,Z=1))))
    WeightMinimum=0
    Weight=@n8.OutputParam
   )
   > n9 = &RigUnit_SphericalPoseReader(
    ActiveRegionScaleFactors=(PositiveWidth=1.5,NegativeWidth=1.2,PositiveHeight=1.3,NegativeHeight=1.2)
    ActiveRegionSize=0.1
    Debug=(bDrawDebug=false,bDraw2D=False,bDrawLocalAxes=False,DebugScale=25,DebugSegments=20,DebugThickness=0.25)
    DriverAxis=(X=1,Y=0,Z=0)
    DriverItem=(Type=Bone,Name="upperarm_l")
    FalloffRegionScaleFactors=(PositiveWidth=1.5,NegativeWidth=1,PositiveHeight=1.2,NegativeHeight=1)
    FalloffSize=0.21
    FlipHeightScaling=False
    FlipWidthScaling=False
    OptionalParentItem=(Type=Bone,Name="None")
    RotationOffset=(X=60,Y=20,Z=-110)
   )
   > &RigUnit_ModifyTransforms(
    ItemToModify=((Item=(Type=Bone,Name="dyn_collar_l_1"),Transform=(Rotation=(X=0,Y=-0.0436,Z=0,W=0.999),Translation=(X=0,Y=0,Z=1),Scale3D=(X=1,Y=1,Z=1))))
    WeightMinimum=0
    Weight=@n9.OutputParam
   )
   > &RigUnit_ModifyTransforms(
    ItemToModify=((Item=(Type=Bone,Name="dyn_collar_l_2"),Transform=(Rotation=(X=0,Y=-0.1736,Z=0,W=0.9848),Translation=(X=0,Y=0,Z=1),Scale3D=(X=1,Y=1,Z=1))))
    WeightMinimum=0
    Weight=@n9.OutputParam
   )
   > n10 = &RigUnit_SphericalPoseReader(
    ActiveRegionScaleFactors=(PositiveWidth=1,NegativeWidth=1,PositiveHeight=1,NegativeHeight=1)
    ActiveRegionSize=0.04
    Debug=(bDrawDebug=false,bDraw2D=False,bDrawLocalAxes=False,DebugScale=25,DebugSegments=20,DebugThickness=0.25)
    DriverAxis=(X=1,Y=0,Z=0)
    DriverItem=(Type=Bone,Name="upperarm_l")
    FalloffRegionScaleFactors=(PositiveWidth=1,NegativeWidth=1,PositiveHeight=1,NegativeHeight=1)
    FalloffSize=0.1
    FlipHeightScaling=False
    FlipWidthScaling=False
    OptionalParentItem=(Type=Bone,Name="None")
    RotationOffset=(X=60,Y=0,Z=-130)
   )
   > &RigUnit_ModifyTransforms(
    ItemToModify=((Item=(Type=Bone,Name="dyn_collar_l_1"),Transform=(Rotation=(X=0,Y=-0.342,Z=0,W=0.9397),Translation=(X=0,Y=0,Z=5),Scale3D=(X=1,Y=1,Z=1))))
    WeightMinimum=0
    Weight=@n10.OutputParam
   )
   > n11 = &RigUnit_SphericalPoseReader(
    ActiveRegionScaleFactors=(PositiveWidth=1.5,NegativeWidth=1,PositiveHeight=1,NegativeHeight=1.5)
    ActiveRegionSize=0.15
    Debug=(bDrawDebug=false,bDraw2D=False,bDrawLocalAxes=False,DebugScale=25,DebugSegments=20,DebugThickness=0.25)
    DriverAxis=(X=-1,Y=0,Z=0)
    DriverItem=(Type=Bone,Name="upperarm_r")
    FalloffRegionScaleFactors=(PositiveWidth=1,NegativeWidth=1.2,PositiveHeight=1.2,NegativeHeight=1)
    FalloffSize=0.35
    FlipHeightScaling=False
    FlipWidthScaling=False
    OptionalParentItem=(Type=Bone,Name="None")
    RotationOffset=(X=-85,Y=0,Z=10)
   )
   > &RigUnit_ModifyTransforms(
    ItemToModify=((Item=(Type=Bone,Name="dyn_collar_r_1"),Transform=(Rotation=(X=0,Y=-0.5,Z=0,W=0.866),Translation=(X=-5.4554,Y=0,Z=0),Scale3D=(X=1,Y=1,Z=1))))
    WeightMinimum=0
    Weight=@n11.OutputParam
   )
   > n12 = &RigUnit_SphericalPoseReader(
    ActiveRegionScaleFactors=(PositiveWidth=2,NegativeWidth=1,PositiveHeight=2,NegativeHeight=1.8)
    ActiveRegionSize=0.05
    Debug=(bDrawDebug=false,bDraw2D=False,bDrawLocalAxes=false,DebugScale=25,DebugSegments=20,DebugThickness=0.25)
    DriverAxis=(X=-1,Y=0,Z=0)
    DriverItem=(Type=Bone,Name="upperarm_r")
    FalloffRegionScaleFactors=(PositiveWidth=1,NegativeWidth=1,PositiveHeight=1,NegativeHeight=1)
    FalloffSize=0.2
    FlipHeightScaling=False
    FlipWidthScaling=False
    OptionalParentItem=(Type=Bone,Name="None")
    RotationOffset=(X=-33,Y=42,Z=15)
   )
   > &RigUnit_ModifyTransforms(
    ItemToModify=((Item=(Type=Bone,Name="dyn_collar_r_1"),Transform=(Rotation=(X=0,Y=0,Z=0,W=-1),Translation=(X=1.0023,Y=-2.6832,Z=1),Scale3D=(X=1,Y=1,Z=1))))
    WeightMinimum=0
    Weight=@n12.OutputParam
   )
   > &RigUnit_ModifyTransforms(
    ItemToModify=((Item=(Type=Bone,Name="dyn_collar_r_2"),Transform=(Rotation=(X=0,Y=0,Z=0,W=1),Translation=(X=0,Y=0,Z=0),Scale3D=(X=1,Y=1,Z=1))))
    WeightMinimum=0
    Weight=@n12.OutputParam
   )
   > n13 = &RigUnit_SphericalPoseReader(
    ActiveRegionScaleFactors=(PositiveWidth=1.2,NegativeWidth=0.5,PositiveHeight=1,NegativeHeight=1.1)
    ActiveRegionSize=0.1
    Debug=(bDrawDebug=false,bDraw2D=False,bDrawLocalAxes=False,DebugScale=25,DebugSegments=20,DebugThickness=0.25)
    DriverAxis=(X=-1,Y=0,Z=0)
    DriverItem=(Type=Bone,Name="upperarm_r")
    FalloffRegionScaleFactors=(PositiveWidth=0.5,NegativeWidth=0.4,PositiveHeight=0,NegativeHeight=1)
    FalloffSize=0.2
    FlipHeightScaling=False
    FlipWidthScaling=False
    OptionalParentItem=(Type=Bone,Name="None")
    RotationOffset=(X=3,Y=49,Z=0)
   )
   > &RigUnit_ModifyTransforms(
    ItemToModify=((Item=(Type=Bone,Name="dyn_collar_r_1"),Transform=(Rotation=(X=0,Y=-0.2164,Z=0,W=0.9763),Translation=(X=0,Y=-0.5,Z=2),Scale3D=(X=1,Y=1,Z=1))))
    WeightMinimum=0
    Weight=@n13.OutputParam
   )
   > n14 = &RigUnit_SphericalPoseReader(
    ActiveRegionScaleFactors=(PositiveWidth=1.1,NegativeWidth=1,PositiveHeight=1.4,NegativeHeight=1)
    ActiveRegionSize=0.15
    Debug=(bDrawDebug=false,bDraw2D=False,bDrawLocalAxes=false,DebugScale=25,DebugSegments=20,DebugThickness=0.25)
    DriverAxis=(X=-1,Y=0,Z=0)
    DriverItem=(Type=Bone,Name="upperarm_r")
    FalloffRegionScaleFactors=(PositiveWidth=1,NegativeWidth=1,PositiveHeight=1,NegativeHeight=1)
    FalloffSize=0.28
    FlipHeightScaling=False
    FlipWidthScaling=False
    OptionalParentItem=(Type=Bone,Name="None")
    RotationOffset=(X=-20,Y=135,Z=0)
   )
   > &RigUnit_ModifyTransforms(
    ItemToModify=((Item=(Type=Bone,Name="dyn_collar_r_1"),Transform=(Rotation=(X=0.2588,Y=0,Z=0,W=0.9659),Translation=(X=-3,Y=0,Z=2),Scale3D=(X=1,Y=1,Z=1))))
    WeightMinimum=0
    Weight=@n14.OutputParam
   )
   > &RigUnit_ModifyTransforms(
    ItemToModify=((Item=(Type=Bone,Name="dyn_collar_r_2"),Transform=(Rotation=(X=0,Y=0,Z=0,W=1),Translation=(X=-3,Y=0,Z=5),Scale3D=(X=1,Y=1,Z=1))))
    Weight=1
    WeightMinimum=0
   )
   > &RigUnit_ModifyTransforms(
    ItemToModify=((Item=(Type=Bone,Name="deform_sleeve_aim_compensation_fr_r"),Transform=(Rotation=(X=0,Y=0,Z=0,W=1),Translation=(X=0,Y=-2.194,Z=-0.1543),Scale3D=(X=1,Y=1,Z=1))))
    WeightMinimum=0
    Weight=@n12.OutputParam
   )
   > n15 = &RigUnit_SphericalPoseReader(
    ActiveRegionScaleFactors=(PositiveWidth=1.1,NegativeWidth=1,PositiveHeight=1.4,NegativeHeight=1)
    ActiveRegionSize=0.15
    Debug=(bDrawDebug=false,bDraw2D=False,bDrawLocalAxes=false,DebugScale=25,DebugSegments=20,DebugThickness=0.25)
    DriverAxis=(X=-1,Y=0,Z=0)
    DriverItem=(Type=Bone,Name="upperarm_r")
    FalloffRegionScaleFactors=(PositiveWidth=1,NegativeWidth=1,PositiveHeight=1,NegativeHeight=1)
    FalloffSize=0.28
    FlipHeightScaling=False
    FlipWidthScaling=False
    OptionalParentItem=(Type=Bone,Name="None")
    RotationOffset=(X=-20,Y=135,Z=0)
   )
   > &RigUnit_ModifyTransforms(
    ItemToModify=((Item=(Type=Bone,Name="deform_armpit_bk_r"),Transform=(Rotation=(X=0,Y=0.0436,Z=0,W=0.999),Translation=(X=-2.9997,Y=2,Z=-1),Scale3D=(X=1,Y=1,Z=1.2))))
    WeightMinimum=0
    Weight=@n15.OutputParam
   )
   > n16 = &RigUnit_SphericalPoseReader(
    ActiveRegionScaleFactors=(PositiveWidth=1,NegativeWidth=1,PositiveHeight=1,NegativeHeight=1)
    ActiveRegionSize=0.07
    Debug=(bDrawDebug=false,bDraw2D=False,bDrawLocalAxes=false,DebugScale=25,DebugSegments=20,DebugThickness=0.25)
    DriverAxis=(X=-1,Y=0,Z=0)
    DriverItem=(Type=Bone,Name="upperarm_r")
    FalloffRegionScaleFactors=(PositiveWidth=1,NegativeWidth=1,PositiveHeight=1,NegativeHeight=1)
    FalloffSize=0.2
    FlipHeightScaling=False
    FlipWidthScaling=False
    OptionalParentItem=(Type=Bone,Name="None")
    RotationOffset=(X=95,Y=0,Z=30)
   )
   > &RigUnit_ModifyTransforms(
    ItemToModify=((Item=(Type=Bone,Name="dfrm_backwing_r"),Transform=(Rotation=(X=0,Y=0,Z=0,W=1),Translation=(X=0,Y=0,Z=2),Scale3D=(X=1,Y=1,Z=1))))
    WeightMinimum=0
    Weight=@n16.OutputParam
   )
   > &RigUnit_ModifyTransforms(
    ItemToModify=((Item=(Type=Bone,Name="deform_sleeve_aim_compensation_btm_r"),Transform=(Rotation=(X=0,Y=0,Z=0,W=1),Translation=(X=0,Y=0,Z=1),Scale3D=(X=1,Y=1,Z=1))))
    WeightMinimum=0
    Weight=@n16.OutputParam
   )
   > &RigUnit_ModifyTransforms(
    ItemToModify=((Item=(Type=Bone,Name="dfrm_backwing_r"),Transform=(Rotation=(X=0,Y=0,Z=0,W=1),Translation=(X=-1,Y=-5,Z=-1),Scale3D=(X=1,Y=1,Z=1))))
    WeightMinimum=0
    Weight=@n15.OutputParam
   )
   > n17 = &RigUnit_SphericalPoseReader(
    ActiveRegionScaleFactors=(PositiveWidth=2.2,NegativeWidth=1,PositiveHeight=1.2,NegativeHeight=2.3)
    ActiveRegionSize=0.08
    Debug=(bDrawDebug=false,bDraw2D=False,bDrawLocalAxes=False,DebugScale=25,DebugSegments=20,DebugThickness=0.25)
    DriverAxis=(X=-1,Y=0,Z=0)
    DriverItem=(Type=Bone,Name="upperarm_r")
    FalloffRegionScaleFactors=(PositiveWidth=3,NegativeWidth=0.5,PositiveHeight=1.5,NegativeHeight=3)
    FalloffSize=0.22
    FlipHeightScaling=False
    FlipWidthScaling=False
    OptionalParentItem=(Type=Bone,Name="None")
    RotationOffset=(X=-124,Y=230,Z=0)
   )
   > &RigUnit_ModifyTransforms(
    ItemToModify=((Item=(Type=Bone,Name="dfrm_backwing_r"),Transform=(Rotation=(X=0,Y=0,Z=0,W=1),Translation=(X=-1.6061,Y=-3.1604,Z=-3),Scale3D=(X=1,Y=1,Z=1))))
    Weight=1
    WeightMinimum=0
   )
   > &RigUnit_ModifyTransforms(
    ItemToModify=((Item=(Type=Bone,Name="deform_armpit_bk_r"),Transform=(Rotation=(X=0,Y=0,Z=0,W=-1),Translation=(X=-2,Y=0,Z=2),Scale3D=(X=1,Y=1,Z=1))))
    WeightMinimum=0
    Weight=@n17.OutputParam
   )
   > n18 = &RigUnit_SphericalPoseReader(
    ActiveRegionScaleFactors=(PositiveWidth=1,NegativeWidth=2,PositiveHeight=2,NegativeHeight=1)
    ActiveRegionSize=0.14
    Debug=(bDrawDebug=false,bDraw2D=False,bDrawLocalAxes=False,DebugScale=25,DebugSegments=20,DebugThickness=0.25)
    DriverAxis=(X=-1,Y=0,Z=0)
    DriverItem=(Type=Bone,Name="upperarm_r")
    FalloffRegionScaleFactors=(PositiveWidth=1,NegativeWidth=1.3,PositiveHeight=1.3,NegativeHeight=0.7)
    FalloffSize=0.35
    FlipHeightScaling=False
    FlipWidthScaling=False
    OptionalParentItem=(Type=Bone,Name="None")
    RotationOffset=(X=-120,Y=0,Z=30)
   )
   > &RigUnit_ModifyTransforms(
    ItemToModify=((Item=(Type=Bone,Name="dfrm_backwing_r"),Transform=(Rotation=(X=0.2372,Y=-0.1821,Z=-0.1373,W=0.9443),Translation=(X=5,Y=-5,Z=1),Scale3D=(X=1,Y=1,Z=1))))
    WeightMinimum=0
    Weight=@n18.OutputParam
   )
   > &RigUnit_ModifyTransforms(
    ItemToModify=((Item=(Type=Bone,Name="deform_armpit_bk_r"),Transform=(Rotation=(X=0,Y=-0.0872,Z=0,W=0.9962),Translation=(X=-2,Y=1,Z=1),Scale3D=(X=1,Y=1,Z=1))))
    WeightMinimum=0
    Weight=@n18.OutputParam
   )
   > n19 = &RigUnit_SphericalPoseReader(
    ActiveRegionScaleFactors=(PositiveWidth=2,NegativeWidth=1.2,PositiveHeight=1,NegativeHeight=2)
    ActiveRegionSize=0.1
    Debug=(bDrawDebug=false,bDraw2D=False,bDrawLocalAxes=False,DebugScale=25,DebugSegments=20,DebugThickness=0.25)
    DriverAxis=(X=0,Y=-1,Z=0)
    DriverItem=(Type=Bone,Name="upperarm_r")
    FalloffRegionScaleFactors=(PositiveWidth=0.5,NegativeWidth=0.5,PositiveHeight=0.3,NegativeHeight=2)
    FalloffSize=0.25
    FlipHeightScaling=False
    FlipWidthScaling=False
    OptionalParentItem=(Type=Bone,Name="None")
    RotationOffset=(X=-180,Y=0,Z=40)
   )
   > &RigUnit_ModifyTransforms(
    ItemToModify=((Item=(Type=Bone,Name="dfrm_backwing_r"),Transform=(Rotation=(X=0.2509,Y=0.0755,Z=0.1148,W=0.9582),Translation=(X=5,Y=-2,Z=-3),Scale3D=(X=1,Y=1,Z=1))))
    WeightMinimum=0
    Weight=@n19.OutputParam
   )
   > &RigUnit_ModifyTransforms(
    ItemToModify=((Item=(Type=Bone,Name="deform_armpit_bk_r"),Transform=(Rotation=(X=0,Y=0,Z=0,W=1),Translation=(X=2,Y=2.383,Z=-1),Scale3D=(X=1,Y=1,Z=1))))
    WeightMinimum=0
    Weight=@n19.OutputParam
   )
   > n20 = &RigUnit_SphericalPoseReader(
    ActiveRegionScaleFactors=(PositiveWidth=1,NegativeWidth=1.3,PositiveHeight=1,NegativeHeight=1)
    ActiveRegionSize=0.1
    Debug=(bDrawDebug=true,bDraw2D=False,bDrawLocalAxes=False,DebugScale=25,DebugSegments=20,DebugThickness=0.25)
    DriverAxis=(X=0,Y=-1,Z=0)
    DriverItem=(Type=Bone,Name="upperarm_r")
    FalloffRegionScaleFactors=(PositiveWidth=1,NegativeWidth=1.3,PositiveHeight=1,NegativeHeight=1)
    FalloffSize=0.15
    FlipHeightScaling=False
    FlipWidthScaling=False
    OptionalParentItem=(Type=Bone,Name="None")
    RotationOffset=(X=-140,Y=0,Z=0)
   )
   > &RigUnit_ModifyTransforms(
    ItemToModify=((Item=(Type=Bone,Name="dfrm_backwing_r"),Transform=(Rotation=(X=0,Y=0,Z=0,W=1),Translation=(X=-3,Y=-4,Z=-5),Scale3D=(X=1,Y=1,Z=1))))
    WeightMinimum=0
    Weight=@n20.OutputParam
   )
   > &RigUnit_ModifyTransforms(
    ItemToModify=((Item=(Type=Bone,Name="deform_armpit_bk_r"),Transform=(Rotation=(X=0,Y=0,Z=0,W=-1),Translation=(X=-1,Y=0,Z=0),Scale3D=(X=1,Y=1,Z=1))))
    WeightMinimum=0
    Weight=@n20.OutputParam
   )
   > n21 = &RigUnit_SphericalPoseReader(
    ActiveRegionScaleFactors=(PositiveWidth=1.2,NegativeWidth=0.5,PositiveHeight=1,NegativeHeight=1.1)
    ActiveRegionSize=0.1
    Debug=(bDrawDebug=false,bDraw2D=False,bDrawLocalAxes=False,DebugScale=25,DebugSegments=20,DebugThickness=0.25)
    DriverAxis=(X=-1,Y=0,Z=0)
    DriverItem=(Type=Bone,Name="upperarm_r")
    FalloffRegionScaleFactors=(PositiveWidth=0.5,NegativeWidth=0.4,PositiveHeight=0,NegativeHeight=1)
    FalloffSize=0.2
    FlipHeightScaling=False
    FlipWidthScaling=False
    OptionalParentItem=(Type=Bone,Name="None")
    RotationOffset=(X=3,Y=49,Z=0)
   )
   > &RigUnit_SphericalPoseReader(
    ActiveRegionScaleFactors=(PositiveWidth=1,NegativeWidth=1,PositiveHeight=1,NegativeHeight=1)
    ActiveRegionSize=0.05
    Debug=(bDrawDebug=false,bDraw2D=False,bDrawLocalAxes=False,DebugScale=25,DebugSegments=20,DebugThickness=0.25)
    DriverAxis=(X=0,Y=-1,Z=0)
    DriverItem=(Type=Bone,Name="upperarm_r")
    FalloffRegionScaleFactors=(PositiveWidth=1,NegativeWidth=1,PositiveHeight=1,NegativeHeight=1)
    FalloffSize=0.15
    FlipHeightScaling=False
    FlipWidthScaling=False
    OptionalParentItem=(Type=Bone,Name="None")
    RotationOffset=(X=-100,Y=0,Z=-20)
   )
   > &RigUnit_ModifyTransforms(
    ItemToModify=((Item=(Type=Bone,Name="deform_armpit_bk_r"),Transform=(Rotation=(X=0,Y=0,Z=0,W=-1),Translation=(X=-3,Y=0,Z=2),Scale3D=(X=1,Y=1,Z=1))))
    Weight=1
    WeightMinimum=0
   )
   > &RigUnit_ModifyTransforms(
    ItemToModify=((Item=(Type=Bone,Name="dfrm_backwing_r"),Transform=(Rotation=(X=0,Y=-0.0872,Z=0,W=-0.9962),Translation=(X=-0.9718,Y=-2.3411,Z=3),Scale3D=(X=1,Y=1,Z=1))))
    WeightMinimum=0
    Weight=@n21.OutputParam
   )
   > &RigUnit_ModifyTransforms(
    ItemToModify=((Item=(Type=Bone,Name="deform_armpit_bk_r"),Transform=(Rotation=(X=0,Y=0,Z=0,W=-1),Translation=(X=-5,Y=0,Z=3),Scale3D=(X=1,Y=1,Z=1))))
    WeightMinimum=0
    Weight=@n21.OutputParam
   )
   > n22 = &RigUnit_SphericalPoseReader(
    ActiveRegionScaleFactors=(PositiveWidth=1,NegativeWidth=1,PositiveHeight=1,NegativeHeight=1)
    ActiveRegionSize=0.02
    Debug=(bDrawDebug=false,bDraw2D=False,bDrawLocalAxes=false,DebugScale=25,DebugSegments=20,DebugThickness=0.25)
    DriverAxis=(X=-1,Y=0,Z=0)
    DriverItem=(Type=Bone,Name="upperarm_r")
    FalloffRegionScaleFactors=(PositiveWidth=1,NegativeWidth=1,PositiveHeight=1,NegativeHeight=1)
    FalloffSize=0.1
    FlipHeightScaling=False
    FlipWidthScaling=False
    OptionalParentItem=(Type=Bone,Name="None")
    RotationOffset=(X=-20,Y=40,Z=15)
   )
   > &RigUnit_ModifyTransforms(
    ItemToModify=((Item=(Type=Bone,Name="dfrm_backwing_r"),Transform=(Rotation=(X=0,Y=-0.0872,Z=0,W=-0.9962),Translation=(X=-0.8666,Y=4,Z=3.597),Scale3D=(X=1,Y=1,Z=1))))
    WeightMinimum=0
    Weight=@n22.OutputParam
   )
   > n23 = &RigUnit_SphericalPoseReader(
    ActiveRegionScaleFactors=(PositiveWidth=1,NegativeWidth=1,PositiveHeight=0.8,NegativeHeight=1)
    ActiveRegionSize=0.15
    Debug=(bDrawDebug=false,bDraw2D=False,bDrawLocalAxes=False,DebugScale=25,DebugSegments=20,DebugThickness=0.25)
    DriverAxis=(X=1,Y=0,Z=0)
    DriverItem=(Type=Bone,Name="upperarm_l")
    FalloffRegionScaleFactors=(PositiveWidth=1,NegativeWidth=1,PositiveHeight=0.6,NegativeHeight=1)
    FalloffSize=0.3
    FlipHeightScaling=False
    FlipWidthScaling=False
    OptionalParentItem=(Type=Bone,Name="None")
    RotationOffset=(X=60,Y=0,Z=-40)
   )
   > &RigUnit_ModifyTransforms(
    ItemToModify=((Item=(Type=Bone,Name="dfrm_backwing_l"),Transform=(Rotation=(X=-0.0436,Y=0,Z=0,W=-0.999),Translation=(X=3,Y=14,Z=6),Scale3D=(X=1,Y=1,Z=1))))
    WeightMinimum=0
    Weight=@n23.OutputParam
   )
   > &RigUnit_ModifyTransforms(
    ItemToModify=((Item=(Type=Bone,Name="deform_ADS_upperarm_compensation"),Transform=(Rotation=(X=0.1305,Y=0,Z=0,W=-0.9914),Translation=(X=0,Y=-2.3314,Z=-0.4051),Scale3D=(X=1,Y=1,Z=1))))
    WeightMinimum=0
    Weight=@n23.OutputParam
   )
   > &RigUnit_ModifyTransforms(
    ItemToModify=((Item=(Type=Bone,Name="deform_armpit_bk_l"),Transform=(Rotation=(X=0,Y=0,Z=0,W=-1),Translation=(X=4,Y=0,Z=4),Scale3D=(X=1,Y=1,Z=1))))
    WeightMinimum=0
    Weight=@n23.OutputParam
   )
   > &RigUnit_ModifyTransforms(
    ItemToModify=((Item=(Type=Bone,Name="deform_sleeve_aim_compensation_btm_l"),Transform=(Rotation=(X=0,Y=0,Z=0,W=-1),Translation=(X=0,Y=-1,Z=-2),Scale3D=(X=1,Y=1,Z=1))))
    WeightMinimum=0
    Weight=@n23.OutputParam
   )
   > n24 = &RigUnit_SphericalPoseReader(
    ActiveRegionScaleFactors=(PositiveWidth=1.2,NegativeWidth=1,PositiveHeight=1.5,NegativeHeight=1)
    ActiveRegionSize=0.1
    Debug=(bDrawDebug=false,bDraw2D=false,bDrawLocalAxes=False,DebugScale=25,DebugSegments=20,DebugThickness=0.25)
    DriverAxis=(X=1,Y=0,Z=0)
    DriverItem=(Type=Bone,Name="upperarm_l")
    FalloffRegionScaleFactors=(PositiveWidth=0.5,NegativeWidth=0.5,PositiveHeight=1.5,NegativeHeight=0)
    FalloffSize=0.2
    FlipHeightScaling=False
    FlipWidthScaling=False
    OptionalParentItem=(Type=Bone,Name="None")
    RotationOffset=(X=8,Y=-140,Z=5)
   )
   > &RigUnit_ModifyTransforms(
    ItemToModify=((Item=(Type=Bone,Name="deform_armpit_bk_l"),Transform=(Rotation=(X=0,Y=0,Z=0,W=-1),Translation=(X=5.5042,Y=0,Z=1.9438),Scale3D=(X=1,Y=1,Z=1))))
    WeightMinimum=0
    Weight=@n24.OutputParam
   )
   > &RigUnit_ModifyTransforms(
    ItemToModify=((Item=(Type=Bone,Name="deform_ADS_upperarm_compensation"),Transform=(Rotation=(X=0,Y=0,Z=0,W=-1),Translation=(X=0,Y=-2,Z=0),Scale3D=(X=1,Y=1,Z=1))))
    WeightMinimum=0
    Weight=@n24.OutputParam
   )
   > &RigUnit_ModifyTransforms(
    ItemToModify=((Item=(Type=Bone,Name="dfrm_backwing_l"),Transform=(Rotation=(X=0,Y=0,Z=0,W=1),Translation=(X=-10,Y=3,Z=0),Scale3D=(X=1,Y=1,Z=1))))
    WeightMinimum=0
    Weight=@n24.OutputParam
   )
   > n25 = &RigUnit_SphericalPoseReader(
    ActiveRegionScaleFactors=(PositiveWidth=1.2,NegativeWidth=1.2,PositiveHeight=1.5,NegativeHeight=1.3)
    ActiveRegionSize=0.07
    Debug=(bDrawDebug=false,bDraw2D=False,bDrawLocalAxes=False,DebugScale=25,DebugSegments=20,DebugThickness=0.25)
    DriverAxis=(X=1,Y=0,Z=0)
    DriverItem=(Type=Bone,Name="upperarm_l")
    FalloffRegionScaleFactors=(PositiveWidth=0.7,NegativeWidth=1,PositiveHeight=0.8,NegativeHeight=1)
    FalloffSize=0.15
    FlipHeightScaling=False
    FlipWidthScaling=False
    OptionalParentItem=(Type=Bone,Name="None")
    RotationOffset=(X=10,Y=-105,Z=0)
   )
   > n26 = &RigUnit_SphericalPoseReader(
    ActiveRegionScaleFactors=(PositiveWidth=1,NegativeWidth=1,PositiveHeight=1,NegativeHeight=1)
    ActiveRegionSize=0.07
    Debug=(bDrawDebug=false,bDraw2D=False,bDrawLocalAxes=False,DebugScale=25,DebugSegments=20,DebugThickness=0.25)
    DriverAxis=(X=0,Y=0,Z=1)
    DriverItem=(Type=Bone,Name="upperarm_l")
    FalloffRegionScaleFactors=(PositiveWidth=1,NegativeWidth=1,PositiveHeight=1,NegativeHeight=1)
    FalloffSize=0.16
    FlipHeightScaling=False
    FlipWidthScaling=False
    OptionalParentItem=(Type=Bone,Name="None")
    RotationOffset=(X=0,Y=-20,Z=0)
   )
   > &RigUnit_ModifyTransforms(
    ItemToModify=((Item=(Type=Bone,Name="dfrm_backwing_l"),Transform=(Rotation=(X=0,Y=0,Z=0,W=-1),Translation=(X=-2.9137,Y=0,Z=-1),Scale3D=(X=1,Y=1,Z=1))))
    WeightMinimum=0
    Weight=@n25.OutputParam
   )
   > &RigUnit_ModifyTransforms(
    ItemToModify=((Item=(Type=Bone,Name="deform_armpit_bk_l"),Transform=(Rotation=(X=0,Y=0,Z=0,W=-1),Translation=(X=6.2917,Y=-0.1245,Z=5),Scale3D=(X=1,Y=1,Z=1))))
    WeightMinimum=0
    Weight=@n25.OutputParam
   )
   > &RigUnit_ModifyTransforms(
    ItemToModify=((Item=(Type=Bone,Name="dfrm_backwing_l"),Transform=(Rotation=(X=0,Y=-0.0872,Z=0,W=0.9962),Translation=(X=5,Y=-1,Z=-1),Scale3D=(X=1,Y=1,Z=1))))
    WeightMinimum=0
    Weight=@n26.OutputParam
   )
   > &RigUnit_ModifyTransforms(
    ItemToModify=((Item=(Type=Bone,Name="deform_armpit_bk_l"),Transform=(Rotation=(X=0,Y=0,Z=0,W=-1),Translation=(X=-6,Y=0,Z=8),Scale3D=(X=1,Y=1,Z=1))))
    WeightMinimum=0
    Weight=@n26.OutputParam
   )
   > &RigUnit_ModifyTransforms(
    ItemToModify=((Item=(Type=Bone,Name="deform_ADS_upperarm_compensation"),Transform=(Rotation=(X=0,Y=0,Z=0,W=-1),Translation=(X=0,Y=-1.9722,Z=0),Scale3D=(X=1,Y=1,Z=1))))
    WeightMinimum=0
    Weight=@n25.OutputParam
   )
   > &RigUnit_ModifyTransforms(
    ItemToModify=((Item=(Type=Bone,Name="deform_sleeve_aim_compensation_fr_l"),Transform=(Rotation=(X=0,Y=0,Z=-0.0436,W=0.999),Translation=(X=-3.5651,Y=-0.2134,Z=-0.7765),Scale3D=(X=1,Y=1,Z=1))))
    WeightMinimum=0
    Weight=@n25.OutputParam
   )
   > n27 = &RigUnit_SphericalPoseReader(
    ActiveRegionScaleFactors=(PositiveWidth=1.2,NegativeWidth=1,PositiveHeight=1.5,NegativeHeight=1)
    ActiveRegionSize=0.1
    Debug=(bDrawDebug=false,bDraw2D=False,bDrawLocalAxes=False,DebugScale=25,DebugSegments=20,DebugThickness=0.25)
    DriverAxis=(X=0,Y=0,Z=1)
    DriverItem=(Type=Bone,Name="upperarm_l")
    FalloffRegionScaleFactors=(PositiveWidth=1.2,NegativeWidth=1,PositiveHeight=1.5,NegativeHeight=1)
    FalloffSize=0.2
    FlipHeightScaling=False
    FlipWidthScaling=False
    OptionalParentItem=(Type=Bone,Name="None")
    RotationOffset=(X=-80,Y=-110,Z=5)
   )
   > &RigUnit_ModifyTransforms(
    ItemToModify=((Item=(Type=Bone,Name="dfrm_backwing_l"),Transform=(Rotation=(X=0,Y=-0.0872,Z=0,W=0.9962),Translation=(X=-4,Y=-2,Z=10),Scale3D=(X=1,Y=1,Z=1))))
    WeightMinimum=0
    Weight=@n27.OutputParam
   )
   > &RigUnit_ModifyTransforms(
    ItemToModify=((Item=(Type=Bone,Name="deform_armpit_bk_l"),Transform=(Rotation=(X=0,Y=0,Z=0,W=-1),Translation=(X=5,Y=8,Z=0),Scale3D=(X=1,Y=1,Z=1))))
    WeightMinimum=0
    Weight=@n27.OutputParam
   )
   > &RigUnit_ModifyTransforms(
    ItemToModify=((Item=(Type=Bone,Name="deform_sleeve_aim_compensation_btm_l"),Transform=(Rotation=(X=0,Y=0,Z=0,W=1),Translation=(X=0,Y=-2,Z=-5),Scale3D=(X=1,Y=1,Z=1))))
    WeightMinimum=0
    Weight=@n27.OutputParam
   )
   > n28 = &RigUnit_SphericalPoseReader(
    ActiveRegionScaleFactors=(PositiveWidth=0.6,NegativeWidth=2,PositiveHeight=2,NegativeHeight=0.5)
    ActiveRegionSize=0.05
    Debug=(bDrawDebug=false,bDraw2D=False,bDrawLocalAxes=False,DebugScale=25,DebugSegments=20,DebugThickness=0.25)
    DriverAxis=(X=1,Y=0,Z=0)
    DriverItem=(Type=Bone,Name="upperarm_l")
    FalloffRegionScaleFactors=(PositiveWidth=0.5,NegativeWidth=1.3,PositiveHeight=1.3,NegativeHeight=0.4)
    FalloffSize=0.15
    FlipHeightScaling=False
    FlipWidthScaling=False
    OptionalParentItem=(Type=Bone,Name="None")
    RotationOffset=(X=40,Y=-130,Z=0)
   )
   > &RigUnit_ModifyTransforms(
    ItemToModify=((Item=(Type=Bone,Name="dfrm_backwing_l"),Transform=(Rotation=(X=0,Y=0,Z=0,W=1),Translation=(X=0,Y=4.3446,Z=0.6767),Scale3D=(X=1,Y=1,Z=1))))
    WeightMinimum=0
    Weight=@n28.OutputParam
   )
   > &RigUnit_ModifyTransforms(
    ItemToModify=((Item=(Type=Bone,Name="deform_armpit_bk_l"),Transform=(Rotation=(X=0,Y=0,Z=0,W=1),Translation=(X=3.5413,Y=0,Z=0),Scale3D=(X=1,Y=1,Z=1))))
    WeightMinimum=0
    Weight=@n28.OutputParam
   )
   > n29 = &RigUnit_SphericalPoseReader(
    ActiveRegionScaleFactors=(PositiveWidth=1,NegativeWidth=2,PositiveHeight=0.8,NegativeHeight=1.8)
    ActiveRegionSize=0.08
    Debug=(bDrawDebug=false,bDraw2D=False,bDrawLocalAxes=False,DebugScale=25,DebugSegments=20,DebugThickness=0.25)
    DriverAxis=(X=1,Y=0,Z=0)
    DriverItem=(Type=Bone,Name="upperarm_l")
    FalloffRegionScaleFactors=(PositiveWidth=0.4,NegativeWidth=2,PositiveHeight=0.4,NegativeHeight=1)
    FalloffSize=0.2
    FlipHeightScaling=False
    FlipWidthScaling=False
    OptionalParentItem=(Type=Bone,Name="None")
    RotationOffset=(X=-40,Y=-125,Z=0)
   )
   > &RigUnit_ModifyTransforms(
    ItemToModify=((Item=(Type=Bone,Name="dfrm_backwing_l"),Transform=(Rotation=(X=0,Y=0,Z=0,W=1),Translation=(X=2,Y=-1.153,Z=-5),Scale3D=(X=1,Y=1,Z=1))))
    WeightMinimum=0
    Weight=@n29.OutputParam
   )
   > &RigUnit_ModifyTransforms(
    ItemToModify=((Item=(Type=Bone,Name="deform_armpit_bk_l"),Transform=(Rotation=(X=0,Y=0,Z=0,W=-1),Translation=(X=0,Y=5,Z=5),Scale3D=(X=1,Y=1,Z=1))))
    WeightMinimum=0
    Weight=@n29.OutputParam
   )
   > n30 = &RigUnit_SphericalPoseReader(
    ActiveRegionScaleFactors=(PositiveWidth=1,NegativeWidth=1,PositiveHeight=1,NegativeHeight=1)
    ActiveRegionSize=0.08
    Debug=(bDrawDebug=false,bDraw2D=False,bDrawLocalAxes=False,DebugScale=25,DebugSegments=20,DebugThickness=0.25)
    DriverAxis=(X=1,Y=0,Z=0)
    DriverItem=(Type=Bone,Name="upperarm_l")
    FalloffRegionScaleFactors=(PositiveWidth=1,NegativeWidth=1,PositiveHeight=1,NegativeHeight=1)
    FalloffSize=0.14
    FlipHeightScaling=False
    FlipWidthScaling=False
    OptionalParentItem=(Type=Bone,Name="None")
    RotationOffset=(X=60,Y=0,Z=-100)
   )
   > &RigUnit_ModifyTransforms(
    ItemToModify=((Item=(Type=Bone,Name="dfrm_backwing_l"),Transform=(Rotation=(X=0,Y=0,Z=0,W=-1),Translation=(X=2.4589,Y=3.026,Z=-4.1183),Scale3D=(X=1,Y=1,Z=1))))
    WeightMinimum=0
    Weight=@n30.OutputParam
   )
   > n31 = &RigUnit_SphericalPoseReader(
    ActiveRegionScaleFactors=(PositiveWidth=1,NegativeWidth=1,PositiveHeight=1,NegativeHeight=1)
    ActiveRegionSize=0.1
    Debug=(bDrawDebug=true,bDraw2D=False,bDrawLocalAxes=False,DebugScale=25,DebugSegments=20,DebugThickness=0.25)
    DriverAxis=(X=0,Y=0,Z=1)
    DriverItem=(Type=Bone,Name="upperarm_l")
    FalloffRegionScaleFactors=(PositiveWidth=1,NegativeWidth=1,PositiveHeight=1,NegativeHeight=1)
    FalloffSize=0.2
    FlipHeightScaling=False
    FlipWidthScaling=False
    OptionalParentItem=(Type=Bone,Name="None")
    RotationOffset=(X=90,Y=0,Z=-110)
   )
   > &RigUnit_ModifyTransforms(
    ItemToModify=((Item=(Type=Bone,Name="dfrm_backwing_l"),Transform=(Rotation=(X=0,Y=0,Z=0,W=-1),Translation=(X=-5,Y=6,Z=21),Scale3D=(X=1,Y=1,Z=1))))
    WeightMinimum=0
    Weight=@n31.OutputParam
   )
   > &RigUnit_ModifyTransforms(
    ItemToModify=((Item=(Type=Bone,Name="deform_armpit_bk_l"),Transform=(Rotation=(X=0,Y=0,Z=0,W=-1),Translation=(X=15,Y=0,Z=-2),Scale3D=(X=1,Y=1,Z=1))))
    WeightMinimum=0
    Weight=@n31.OutputParam
   )
   > &RigUnit_ModifyTransforms(
    ItemToModify=((Item=(Type=Bone,Name="deform_ADS_upperarm_compensation"),Transform=(Rotation=(X=0,Y=0,Z=0,W=-1),Translation=(X=0,Y=-3,Z=0),Scale3D=(X=1,Y=1,Z=1))))
    WeightMinimum=0
    Weight=@n31.OutputParam
   )
   SequenceAggregate_1_2 = RigVMAggregateNode(ExecuteContext=@SequenceAggregate_1.D)
   n32 = &RigUnit_SphericalPoseReader(
    ActiveRegionScaleFactors=(PositiveWidth=1,NegativeWidth=1,PositiveHeight=1.2,NegativeHeight=1.2)
    ActiveRegionSize=0.05
    Debug=(bDrawDebug=false,bDraw2D=False,bDrawLocalAxes=False,DebugScale=25,DebugSegments=20,DebugThickness=0.25)
    DriverAxis=(X=1,Y=0,Z=0)
    DriverItem=(Type=Bone,Name="upperarm_l")
    FalloffRegionScaleFactors=(PositiveWidth=0.5,NegativeWidth=0.75,PositiveHeight=0.8,NegativeHeight=0.5)
    FalloffSize=0.3
    FlipHeightScaling=False
    FlipWidthScaling=False
    OptionalParentItem=(Type=Bone,Name="None")
    RotationOffset=(X=0,Y=-135,Z=15)
    ExecutePin=@SequenceAggregate_1_2.A
   )
   > RigVMVariableNode(Value=@n32.OutputParam) {variable=SPR_upperarm_l_down_A_alpha}
   > n33 = &RigUnit_SphericalPoseReader(
    ActiveRegionScaleFactors=(PositiveWidth=1,NegativeWidth=1,PositiveHeight=1,NegativeHeight=1)
    ActiveRegionSize=0.065
    Debug=(bDrawDebug=false,bDraw2D=False,bDrawLocalAxes=False,DebugScale=25,DebugSegments=20,DebugThickness=0.25)
    DriverAxis=(X=-1,Y=0,Z=0)
    DriverItem=(Type=Bone,Name="upperarm_r")
    FalloffRegionScaleFactors=(PositiveWidth=0.75,NegativeWidth=1,PositiveHeight=0.75,NegativeHeight=1)
    FalloffSize=0.2
    FlipHeightScaling=False
    FlipWidthScaling=False
    OptionalParentItem=(Type=Bone,Name="None")
    RotationOffset=(X=0,Y=50,Z=-7.5)
   )
   > RigVMVariableNode(Value=@n33.OutputParam) {variable=SPR_upperarm_r_down_A_alpha}
   > n34 = &RigUnit_SphericalPoseReader(
    ActiveRegionScaleFactors=(PositiveWidth=1,NegativeWidth=1,PositiveHeight=1,NegativeHeight=1)
    ActiveRegionSize=0.05
    Debug=(bDrawDebug=false,bDraw2D=False,bDrawLocalAxes=true,DebugScale=25,DebugSegments=20,DebugThickness=0.25)
    DriverAxis=(X=-1,Y=0,Z=0)
    DriverItem=(Type=Bone,Name="upperarm_r")
    FalloffRegionScaleFactors=(PositiveWidth=0.5,NegativeWidth=1,PositiveHeight=0.75,NegativeHeight=1)
    FalloffSize=0.2
    FlipHeightScaling=False
    FlipWidthScaling=False
    OptionalParentItem=(Type=Bone,Name="None")
    RotationOffset=(X=-30,Y=25,Z=0)
   )
   > RigVMVariableNode(Value=@n34.OutputParam) {variable=SPR_upperarm_r_down_B_alpha}
   SequenceAggregate = &RigVMFunction_Sequence(ExecuteContext=@SequenceAggregate_1_2.C)
   RigUnit_ItemArray = &RigUnit_ItemArray(Items=((Type=Bone,Name="clavicle_l"),(Type=Bone,Name="upperarm_l"),(Type=Bone,Name="lowerarm_l")))
   DISPATCH_RigVMDispatch_ArrayIterator = RigVMDispatchNode(Array=@RigUnit_ItemArray.Items, ExecuteContext=@SequenceAggregate.A) {template="DISPATCH_RigVMDispatch_ArrayIterator(in Array,out Element,out Index,out Count,out Ratio)"}
   DISPATCH_RigVMDispatch_ArrayGetAtIndex = RigVMDispatchNode(
    Array=&RigUnit_ItemArray(Items=((Type=Bone,Name="dfrm_clavicle_l"),(Type=Bone,Name="dfrm_upperarm_l"),(Type=Bone,Name="dfrm_lowerarm_l"))).Items
    Index=@DISPATCH_RigVMDispatch_ArrayIterator.Index
   ) {template="DISPATCH_RigVMDispatch_ArrayGetAtIndex(in Array,in Index,out Element)"}
   > &RigUnit_SetRotation(
    Space=GlobalSpace
    Weight=1
    bInitial=False
    bPropagateToChildren=true
    Item=@DISPATCH_RigVMDispatch_ArrayGetAtIndex.Element
    Value=&RigUnit_GetTransform(Item=@DISPATCH_RigVMDispatch_ArrayIterator.Element).Transform.Rotation
   )
   &RigUnit_ModifyTransforms(
    ItemToModify=((Item=(Type=Bone,Name="dfrm_clavicle_l"),Transform=(Rotation=(X=0,Y=0,Z=0,W=1),Translation=(X=1,Y=0,Z=0),Scale3D=(X=1,Y=1,Z=1))))
    WeightMinimum=0
    Weight=RigVMVariableNode() {variable=SPR_upperarm_l_down_A_alpha}.Value
    ExecutePin=@DISPATCH_RigVMDispatch_ArrayIterator.Completed
   )
   > PBIK = /Script/PBIK.RigUnit_PBIK(
    BoneSettings=((Bone="dfrm_lowerarm_l",X=Locked,Y=Locked,Z=Free,PreferredAngles=(X=0,Y=0,Z=0)))
    Debug=(DrawScale=1)
    Effectors=((Bone="dfrm_hand_l",Transform=(Rotation=(X=0,Y=0,Z=0,W=1),Translation=(X=0,Y=0,Z=0),Scale3D=(X=1,Y=1,Z=1)),PositionAlpha=1,RotationAlpha=1,StrengthAlpha=1,PinRotation=1,ChainDepth=0))
    ExcludedBones=()
    Root=dfrm_clavicle_l
    Settings="(Iterations=20,SubIterations=0,MassMultiplier=1,bAllowStretch=False,RootBehavior=PinToInput,PrePullRootSettings=(RotationAlpha=0,RotationAlphaX=1,RotationAlphaY=1,RotationAlphaZ=1,PositionAlpha=1,PositionAlphaX=1,PositionAlphaY=1,PositionAlphaZ=1),GlobalPu ...(truncated)"
    Effectors.0.Transform=&RigUnit_GetTransform(Item=(Type=Bone,Name="hand_l")).Transform
   ) {is=struct RigVMUnitNode}
   n35 = RigVMDispatchNode(
    Array=&RigUnit_ItemArray(Items=((Type=Bone,Name="clavicle_r"),(Type=Bone,Name="upperarm_r"),(Type=Bone,Name="lowerarm_r"))).Items
    ExecuteContext=@SequenceAggregate.B
   ) {template="DISPATCH_RigVMDispatch_ArrayIterator(in Array,out Element,out Index,out Count,out Ratio)"}
   > &RigUnit_SetRotation(
    Space=GlobalSpace
    Weight=1
    bInitial=False
    bPropagateToChildren=true
    Item=RigVMDispatchNode(
     Array=&RigUnit_ItemArray(Items=((Type=Bone,Name="dfrm_clavicle_r"),(Type=Bone,Name="dfrm_upperarm_r"),(Type=Bone,Name="dfrm_lowerarm_r"))).Items
     Index=@n35.Index
    ) {template="DISPATCH_RigVMDispatch_ArrayGetAtIndex(in Array,in Index,out Element)"}.Element
    Value=&RigUnit_GetTransform(Item=@n35.Element).Transform.Rotation
   )
   &RigUnit_ModifyTransforms(
    ItemToModify=((Item=(Type=Bone,Name="dfrm_clavicle_r"),Transform=(Rotation=(X=0,Y=0,Z=0,W=1),Translation=(X=-2,Y=0,Z=0),Scale3D=(X=1,Y=1,Z=1))))
    WeightMinimum=0
    Weight=RigVMVariableNode() {variable=SPR_upperarm_r_down_B_alpha}.Value
    ExecutePin=@n35.Completed
   )
   > &RigUnit_PBIK(
    BoneSettings=((Bone="dfrm_lowerarm_r",X=Locked,Y=Locked,Z=Free,PreferredAngles=(X=0,Y=0,Z=0)))
    Debug=(DrawScale=1)
    Effectors=((Bone="dfrm_hand_r",Transform=(Rotation=(X=0,Y=0,Z=0,W=1),Translation=(X=0,Y=0,Z=0),Scale3D=(X=1,Y=1,Z=1)),PositionAlpha=1,RotationAlpha=1,StrengthAlpha=1,PinRotation=1,ChainDepth=0))
    ExcludedBones=()
    Root=dfrm_clavicle_l
    Settings="(Iterations=20,SubIterations=0,MassMultiplier=1,bAllowStretch=False,RootBehavior=PinToInput,PrePullRootSettings=(RotationAlpha=0,RotationAlphaX=1,RotationAlphaY=1,RotationAlphaZ=1,PositionAlpha=1,PositionAlphaX=1,PositionAlphaY=1,PositionAlphaZ=1),GlobalPu ...(truncated)"
    Effectors.0.Transform=&RigUnit_GetTransform(Item=(Type=Bone,Name="hand_r")).Transform
   )
   SequenceAggregate_2 = &RigVMFunction_Sequence(ExecuteContext=@SequenceAggregate_1_2.D)
   n36 = /Script/RigVM.RigVMFunction_MathVectorUnit(Value=/Script/RigVM.RigVMFunction_MathVectorSub(
     B=&RigUnit_GetTransform(Item=(Type=Bone,Name="dfrm_upperarm_l")).Transform.Translation
     A=&RigUnit_GetTransform(Item=(Type=Bone,Name="dfrm_hand_l")).Transform.Translation
    ) {is=struct RigVMUnitNode, template="Subtract::Execute(in A,in B,out Result)"}.Result) {is=struct RigVMUnitNode, template="Unit::Execute(in Value,out Result)"}
   Set Transform_2 = &RigUnit_SetRotation(
    Item=(Type=Bone,Name="dfrm_upperarm_l")
    Space=GlobalSpace
    bInitial=false
    Value=/Script/RigVM.RigVMFunction_MathQuaternionMul(
     B=&RigUnit_GetTransform(Item=(Type=Bone,Name="dfrm_upperarm_l")).Transform.Rotation
     A=/Script/RigVM.RigVMFunction_MathQuaternionFromAxisAndAngle(Angle=0.2, Axis=@n36.Result) {is=struct RigVMUnitNode, template="FromAxisAndAngle::Execute(in Axis,in Angle,out Result)"}.Result
    ) {is=struct RigVMUnitNode, template="Multiply::Execute(in A,in B,out Result)"}.Result
    ExecutePin=@SequenceAggregate_2.A
    Weight=RigVMVariableNode() {variable=mult_spr_dst_lowerarm_r_pelvis_alpha}.Value
   )
   n37 = &RigVMFunction_MathVectorUnit(Value=&RigVMFunction_MathVectorSub(
     B=&RigUnit_GetTransform(Item=(Type=Bone,Name="dfrm_lowerarm_l")).Transform.Translation
     A=&RigUnit_GetTransform(Item=(Type=Bone,Name="dfrm_hand_l")).Transform.Translation
    ) {template="Subtract::Execute(in A,in B,out Result)"}.Result) {template="Unit::Execute(in Value,out Result)"}
   > Set Transform_2_1 = &RigUnit_SetRotation(
    Item=(Type=Bone,Name="dfrm_lowerarm_l")
    Space=GlobalSpace
    bInitial=false
    Value=&RigVMFunction_MathQuaternionMul(
     B=&RigUnit_GetTransform(Item=(Type=Bone,Name="dfrm_lowerarm_l")).Transform.Rotation
     A=&RigVMFunction_MathQuaternionFromAxisAndAngle(Angle=-1.3, Axis=@n37.Result) {template="FromAxisAndAngle::Execute(in Axis,in Angle,out Result)"}.Result
    ) {template="Multiply::Execute(in A,in B,out Result)"}.Result
    Weight=RigVMVariableNode() {variable=mult_spr_dst_lowerarm_r_pelvis_alpha}.Value
   )
   > n38 = &RigUnit_SphericalPoseReader(
    ActiveRegionScaleFactors=(PositiveWidth=3,NegativeWidth=3,PositiveHeight=1,NegativeHeight=3)
    ActiveRegionSize=0.1
    Debug=(bDrawDebug=false,DebugScale=25,DebugSegments=20,DebugThickness=0.25)
    DriverAxis=(X=0,Y=1,Z=0)
    DriverItem=(Type=Bone,Name="head")
    FalloffRegionScaleFactors=(PositiveWidth=1,NegativeWidth=1,PositiveHeight=0.8,NegativeHeight=1)
    FalloffSize=0.4
    FlipHeightScaling=true
    OptionalParentItem=(Type=Bone,Name="pelvis")
    RotationOffset=(X=125,Y=0,Z=0)
   )
   > n39 = &RigUnit_SphericalPoseReader(
    ActiveRegionScaleFactors=(PositiveWidth=1,NegativeWidth=1.3,PositiveHeight=3,NegativeHeight=2)
    ActiveRegionSize=0.05
    Debug=(bDrawDebug=false,bDraw2D=false,bDrawLocalAxes=true,DebugScale=25,DebugSegments=20,DebugThickness=0.25)
    DriverAxis=(X=1,Y=0,Z=0)
    DriverItem=(Type=Bone,Name="upperarm_l")
    FalloffRegionScaleFactors=(PositiveWidth=0.5,NegativeWidth=0.7,PositiveHeight=1,NegativeHeight=1)
    FalloffSize=0.3
    OptionalParentItem=(Type=Bone,Name="spine_04")
    RotationOffset=(X=90,Y=25,Z=-60)
   )
   n40 = &RigVMFunction_MathVectorUnit(Value=&RigVMFunction_MathVectorSub(
     B=&RigUnit_GetTransform(Item=(Type=Bone,Name="dfrm_upperarm_l")).Transform.Translation
     A=&RigUnit_GetTransform(Item=(Type=Bone,Name="dfrm_hand_l")).Transform.Translation
    ) {template="Subtract::Execute(in A,in B,out Result)"}.Result) {template="Unit::Execute(in Value,out Result)"}
   > Set Transform_3 = &RigUnit_SetRotation(
    Item=(Type=Bone,Name="dfrm_upperarm_l")
    Space=GlobalSpace
    bInitial=false
    Value=&RigVMFunction_MathQuaternionMul(
     B=&RigUnit_GetTransform(Item=(Type=Bone,Name="dfrm_upperarm_l")).Transform.Rotation
     A=&RigVMFunction_MathQuaternionFromAxisAndAngle(Angle=-0.35, Axis=@n40.Result) {template="FromAxisAndAngle::Execute(in Axis,in Angle,out Result)"}.Result
    ) {template="Multiply::Execute(in A,in B,out Result)"}.Result
    Weight=&RigVMFunction_MathDoubleMul(B=@n39.OutputParam, A=@n38.OutputParam).Result
   )
   n41 = &RigVMFunction_MathVectorUnit(Value=&RigVMFunction_MathVectorSub(
     B=&RigUnit_GetTransform(Item=(Type=Bone,Name="dfrm_upperarm_r"), bInitial=False).Transform.Translation
     A=&RigUnit_GetTransform(Item=(Type=Bone,Name="dfrm_hand_r"), bInitial=False).Transform.Translation
    ) {template="Subtract::Execute(in A,in B,out Result)"}.Result) {template="Unit::Execute(in Value,out Result)"}
   Set Transform_1 = &RigUnit_SetRotation(
    Item=(Type=Bone,Name="dfrm_upperarm_r")
    Space=GlobalSpace
    bInitial=False
    Value=&RigVMFunction_MathQuaternionMul(
     B=&RigUnit_GetTransform(Item=(Type=Bone,Name="dfrm_upperarm_r"), bInitial=False).Transform.Rotation
     A=&RigVMFunction_MathQuaternionFromAxisAndAngle(Angle=0.1, Axis=@n41.Result) {template="FromAxisAndAngle::Execute(in Axis,in Angle,out Result)"}.Result
    ) {template="Multiply::Execute(in A,in B,out Result)"}.Result
    Weight=RigVMVariableNode() {variable=SPR_upperarm_r_down_B_alpha}.Value
    ExecutePin=@SequenceAggregate_2.B
   )
   > n42 = &RigUnit_SphericalPoseReader(
    ActiveRegionScaleFactors=(PositiveWidth=0.6,NegativeWidth=0.6,PositiveHeight=1,NegativeHeight=1)
    ActiveRegionSize=0.15
    Debug=(bDrawDebug=false,bDraw2D=false,bDrawLocalAxes=true,DebugScale=25,DebugSegments=20,DebugThickness=0.25)
    DriverAxis=(X=-1,Y=0,Z=0)
    DriverItem=(Type=Bone,Name="upperarm_r")
    FalloffRegionScaleFactors=(PositiveWidth=0.6,NegativeWidth=0.6,PositiveHeight=2,NegativeHeight=1)
    FalloffSize=0.45
    FlipHeightScaling=false
    FlipWidthScaling=true
    OptionalParentItem=(Type=Bone,Name="spine_04")
    RotationOffset=(X=-100,Y=25,Z=40)
   )
   n43 = &RigVMFunction_MathVectorUnit(Value=&RigVMFunction_MathVectorSub(
     A=&RigUnit_GetTransform(Item=(Type=Bone,Name="dfrm_hand_r"), bInitial=False).Transform.Translation
     B=&RigUnit_GetTransform(Item=(Type=Bone,Name="dfrm_upperarm_r"), bInitial=False).Transform.Translation
    ) {template="Subtract::Execute(in A,in B,out Result)"}.Result) {template="Unit::Execute(in Value,out Result)"}
   > Set Transform_1_1 = &RigUnit_SetRotation(
    Item=(Type=Bone,Name="dfrm_upperarm_r")
    Space=GlobalSpace
    bInitial=False
    Value=&RigVMFunction_MathQuaternionMul(
     B=&RigUnit_GetTransform(Item=(Type=Bone,Name="dfrm_upperarm_r"), bInitial=False).Transform.Rotation
     A=&RigVMFunction_MathQuaternionFromAxisAndAngle(Angle=0.5, Axis=@n43.Result) {template="FromAxisAndAngle::Execute(in Axis,in Angle,out Result)"}.Result
    ) {template="Multiply::Execute(in A,in B,out Result)"}.Result
    Weight=@n42.OutputParam
   )
   > n44 = &RigUnit_SphericalPoseReader(
    ActiveRegionScaleFactors=(PositiveWidth=0.6,NegativeWidth=0.6,PositiveHeight=1,NegativeHeight=1)
    ActiveRegionSize=0.15
    Debug=(bDrawDebug=false,bDraw2D=false,bDrawLocalAxes=true,DebugScale=25,DebugSegments=20,DebugThickness=0.25)
    DriverAxis=(X=1,Y=0,Z=0)
    DriverItem=(Type=Bone,Name="upperarm_l")
    FalloffRegionScaleFactors=(PositiveWidth=0.6,NegativeWidth=0.6,PositiveHeight=1,NegativeHeight=1)
    FalloffSize=0.45
    OptionalParentItem=(Type=Bone,Name="spine_04")
    RotationOffset=(X=40,Y=25,Z=-40)
   )
   n45 = &RigVMFunction_MathVectorUnit(Value=&RigVMFunction_MathVectorSub(
     A=&RigUnit_GetTransform(Item=(Type=Bone,Name="dfrm_hand_l"), bInitial=False).Transform.Translation
     B=&RigUnit_GetTransform(Item=(Type=Bone,Name="dfrm_upperarm_l"), bInitial=False).Transform.Translation
    ) {template="Subtract::Execute(in A,in B,out Result)"}.Result) {template="Unit::Execute(in Value,out Result)"}
   > Set Transform_1_1_1 = &RigUnit_SetRotation(
    Item=(Type=Bone,Name="dfrm_upperarm_twist_01_l")
    Space=GlobalSpace
    bInitial=False
    Value=&RigVMFunction_MathQuaternionMul(
     B=&RigUnit_GetTransform(Item=(Type=Bone,Name="dfrm_upperarm_l"), bInitial=False).Transform.Rotation
     A=&RigVMFunction_MathQuaternionFromAxisAndAngle(Angle=-0.4, Axis=@n45.Result) {template="FromAxisAndAngle::Execute(in Axis,in Angle,out Result)"}.Result
    ) {template="Multiply::Execute(in A,in B,out Result)"}.Result
    Weight=@n44.OutputParam
   )
   n46 = &RigVMFunction_MathQuaternionSwingTwist(
    TwistAxis=(X=1,Y=0,Z=0)
    Input=&RigVMFunction_MathTransformMakeRelative(
     Parent=&RigUnit_GetTransform(Item=(Type=Bone,Name="hand_l"), Space=LocalSpace, bInitial=true).Transform
     Global=&RigUnit_GetTransform(Item=(Type=Bone,Name="hand_l"), Space=LocalSpace).Transform
    ).Local.Rotation
   )
   &RigUnit_SetRotation(
    Item=(Type=Bone,Name="dfrm_lowerarm_twist_01_l")
    Weight=0.75
    bPropagateToChildren=false
    Value=@n46.Twist
    ExecutePin=@SequenceAggregate_1_2.E
   )
   > &RigUnit_SetRotation(Item=(Type=Bone,Name="dfrm_lowerarm_twist_02_l"), Weight=0.5, bPropagateToChildren=false, Value=@n46.Twist)
   n47 = &RigVMFunction_MathQuaternionSwingTwist(
    TwistAxis=(X=1,Y=0,Z=0)
    Input=&RigVMFunction_MathTransformMakeRelative(
     Parent=&RigUnit_GetTransform(Item=(Type=Bone,Name="hand_r"), Space=LocalSpace, bInitial=true).Transform
     Global=&RigUnit_GetTransform(Item=(Type=Bone,Name="hand_r"), Space=LocalSpace).Transform
    ).Local.Rotation
   )
   > &RigUnit_SetRotation(Item=(Type=Bone,Name="dfrm_lowerarm_twist_01_r"), Weight=0.75, bPropagateToChildren=false, Value=@n47.Twist)
   > &RigUnit_SetRotation(Item=(Type=Bone,Name="dfrm_lowerarm_twist_02_r"), Weight=0.5, bPropagateToChildren=false, Value=@n47.Twist)
   Remap = &RigVMFunction_MathDoubleRemap(
    SourceMaximum=40
    SourceMinimum=25
    TargetMaximum=1
    TargetMinimum=0
    Value=&RigVMFunction_MathVectorDistance(
     A=&RigUnit_GetTransform(Item=(Type=Bone,Name="pelvis"), bInitial=False).Transform.Translation
     B=&RigUnit_GetTransform(Item=(Type=Bone,Name="lowerarm_l"), bInitial=False).Transform.Translation
    ).Result
   )
   Multiply = &RigVMFunction_MathDoubleMul(B=&RigVMFunction_MathDoubleSub(A=1, B=@Remap.Result).Result, A=RigVMVariableNode() {variable=SPR_upperarm_l_down_A_alpha}.Value)
   RigVMVariableNode(Value=@Multiply.Result, ExecuteContext=@SequenceAggregate_1_2.B) {variable=mult_spr_dst_lowerarm_r_pelvis_alpha}
   RigVMFunction_MathVectorDistance = &RigVMFunction_MathVectorDistance(
    A=&RigUnit_GetTransform(Item=(Type=Bone,Name="hand_r"), bInitial=False).Transform.Translation
    B=&RigUnit_GetTransform(Item=(Type=Bone,Name="thigh_twist_01_r"), bInitial=False).Transform.Translation
   )
   > RigVMVariableNode(Value=&RigVMFunction_MathDoubleSub(
     A=1
     B=&RigVMFunction_MathDoubleRemap(SourceMaximum=59, SourceMinimum=19.5, TargetMaximum=1, TargetMinimum=0, Value=@RigVMFunction_MathVectorDistance.Result).Result
    ).Result) {variable=mult_spr_dst_hand_r_thigh_twist_r_alpha}
  ]}
 SequenceAggregate_1_1_ContainedGraph {nodes [
   n1 = RigVMFunctionEntryNode()
   SequenceAggregate_1 = &RigVMFunction_Sequence(ExecuteContext=@n1.ExecuteContext)
   SequenceAggregate_1_1 = &RigVMFunction_Sequence(ExecuteContext=@SequenceAggregate_1.B)
   SequenceAggregate_1_2 = &RigVMFunction_Sequence(ExecuteContext=@SequenceAggregate_1_1.B)
   SequenceAggregate_1_3 = &RigVMFunction_Sequence(ExecuteContext=@SequenceAggregate_1_2.B)
   SequenceAggregate_1_4 = &RigVMFunction_Sequence(ExecuteContext=@SequenceAggregate_1_3.B)
   RigVMFunctionReturnNode(
    A=@SequenceAggregate_1.A
    B=@SequenceAggregate_1_1.A
    C=@SequenceAggregate_1_2.A
    D=@SequenceAggregate_1_3.A
    E=@SequenceAggregate_1_4.A
    F=@SequenceAggregate_1_4.B
   )
  ]}
 SequenceAggregate_1_1_ContainedGraph {nodes [
   n1 = RigVMFunctionEntryNode()
   SequenceAggregate_1 = &RigVMFunction_Sequence(ExecuteContext=@n1.ExecuteContext)
   SequenceAggregate_1_1 = &RigVMFunction_Sequence(ExecuteContext=@SequenceAggregate_1.B)
   SequenceAggregate_1_2 = &RigVMFunction_Sequence(ExecuteContext=@SequenceAggregate_1_1.B)
   SequenceAggregate_1_3 = &RigVMFunction_Sequence(ExecuteContext=@SequenceAggregate_1_2.B)
   RigVMFunctionReturnNode(A=@SequenceAggregate_1.A, B=@SequenceAggregate_1_1.A, C=@SequenceAggregate_1_2.A, D=@SequenceAggregate_1_3.A, E=@SequenceAggregate_1_3.B)
  ]}
]
hierarchy [
 root {loc=<0 0 0>, rot=<0 0 0 1>}
 pelvis {parent=root, loc=<0 -0.5301 98.0571>, rot=<0.0127 0.707 -0.0127 -0.707>}
 spine_01 {parent=pelvis, loc=<6.1208 -0.9328 0>, rot=<0 0 0.034 -0.9994>}
 spine_02 {parent=spine_01, loc=<6.5996 -0.2095 0>, rot=<0 0 0.0238 -0.9997>}
 spine_03 {parent=spine_02, loc=<6.4914 -0.1237 0>, rot=<0 0 0.0149 -0.9999>}
 spine_04 {parent=spine_03, loc=<6.9598 0.0794 0>, rot=<0 0 -0.0132 -0.9999>}
 spine_05 {parent=spine_04, loc=<6.9987 0 0>, rot=<0 0 -0.0015 -1>}
 neck_01 {parent=spine_05, loc=<11.3138 0.1529 0>, rot=<0 0 -0.2003 -0.9797>}
 neck_02 {parent=neck_01, loc=<4.5627 0 0>, rot=<0 0 0.0504 -0.9987>}
 head {parent=neck_02, loc=<5.3914 0 0>, rot=<0 0 0.115 -0.9934>}
 faceAttach {parent=head, loc=<5.1819 -0.7774 0>, rot=<0.0029 -0.7071 -0.0029 -0.7071>}
 deform_glasses {parent=head, loc=<21.234 1.7891 0>, rot=<0 0 0 1>}
 clavicle_l {parent=spine_05, loc=<5.4831 -0.1053 -3.2946>, rot=<0.0465 -0.7289 -0.009 -0.6829>}
 upperarm_l {parent=clavicle_l, loc=<9.8658 -1.2136 -0.4105>, rot=<0.0164 -0.3245 -0.0269 -0.9454>}
 upperarm_twist_01_l {parent=upperarm_l, loc=<7.2135 -0.1134 0>, rot=<0 0 0 1>}
 upperarm_twist_02_l {parent=upperarm_l, loc=<19.7998 -0.1134 0>, rot=<0 0 0 1>}
 lowerarm_l {parent=upperarm_l, loc=<26.465 -0.8345 0.259>, rot=<0.0036 0.0161 -0.2298 -0.9731>}
 hand_l {parent=lowerarm_l, loc=<22.7355 -0.3203 -0.6105>, rot=<-0.6784 -0.0223 0.0735 -0.7307>}
 clavicle_r {parent=spine_05, loc=<5.4831 -0.1053 3.2946>, rot=<0.7289 0.0465 -0.6829 0.009>}
 upperarm_r {parent=clavicle_r, loc=<-9.8658 1.2136 0.4105>, rot=<0.0164 -0.3245 -0.0269 -0.9454>}
 upperarm_twist_01_r {parent=upperarm_r, loc=<-7.2135 0.1134 0>, rot=<0 0 0 -1>}
 upperarm_twist_02_r {parent=upperarm_r, loc=<-19.7998 0.1134 0>, rot=<0 0 0 -1>}
 lowerarm_r {parent=upperarm_r, loc=<-26.465 0.8345 -0.259>, rot=<0.0036 0.0161 -0.2298 -0.9731>}
 hand_r {parent=lowerarm_r, loc=<-22.7355 0.3203 0.6106>, rot=<-0.6784 -0.0223 0.0735 -0.7307>}
 dfrm_clavicle_l {parent=spine_05, loc=<5.4831 -0.1053 -3.2946>, rot=<0.0465 -0.7289 -0.009 -0.6829>}
 dfrm_upperarm_l {parent=dfrm_clavicle_l, loc=<9.8658 -1.2136 -0.4105>, rot=<0.0164 -0.3245 -0.0269 -0.9454>}
 dfrm_upperarm_twist_01_l {parent=dfrm_upperarm_l, loc=<7.2135 -0.1134 0>, rot=<0 0 0 1>}
 dyn_wing_upper_l_twist_MSTR {parent=dfrm_upperarm_twist_01_l, loc=<0 0 0>, rot=<0 0 0 1>}
 dyn_wing_upper_l_MSTR {parent=dyn_wing_upper_l_twist_MSTR, loc=<2.858 -3.3049 -3.137>, rot=<-0.0599 -0.3671 0.8988 0.2321>}
 dyn_wing_upper_l_01 {parent=dyn_wing_upper_l_MSTR, loc=<-13.8175 -2.74 -1.3027>, rot=<-0.0902 0.1199 -0.1208 -0.9813>}
 dyn_wing_upper_l_02 {parent=dyn_wing_upper_l_MSTR, loc=<-7.1189 1.3474 -0.9037>, rot=<0 0 0 1>}
 dyn_wing_upper_l_02_end {parent=dyn_wing_upper_l_02, loc=<-6.8685 0 0>, rot=<-0.0087 0.0178 -0.0208 -0.9996>}
 dyn_wing_upper_l_03 {parent=dyn_wing_upper_l_MSTR, loc=<-1.5405 4.0704 -0.5145>, rot=<0.0076 0.0232 0.1055 -0.9941>}
 dyn_wing_upper_l_03_end {parent=dyn_wing_upper_l_03, loc=<-6.0908 0 0>, rot=<-0.0222 0.0342 -0.0093 -0.9991>}
 dfrm_upperarm_twist_02_l {parent=dfrm_upperarm_l, loc=<19.7998 -0.1134 0>, rot=<0 0 0 1>}
 dfrm_lowerarm_l {parent=dfrm_upperarm_l, loc=<26.465 -0.8345 0.259>, rot=<0.0036 0.0161 -0.2298 -0.9731>}
 dfrm_lowerarm_twist_02_l {parent=dfrm_lowerarm_l, loc=<10.6707 0 0>, rot=<0 0 0 1>}
 dyn_wing_lower_l_twist_MSTR {parent=dfrm_lowerarm_twist_02_l, loc=<0 0 0>, rot=<0 0 0 1>}
 dyn_wing_lower_l_MSTR {parent=dyn_wing_lower_l_twist_MSTR, loc=<-7.658 -2.048 -0.4383>, rot=<0.0097 -0.1651 0.9668 0.1948>}
 dyn_wing_upper_l_04 {parent=dyn_wing_lower_l_MSTR, loc=<-1.1461 3.2663 -0.3853>, rot=<0.0433 -0.0165 -0.0185 -0.9988>}
 dyn_wing_upper_l_04_end {parent=dyn_wing_upper_l_04, loc=<-9.3847 0.0001 -0.0002>, rot=<0.0023 0.0263 0.0172 -0.9995>}
 dyn_wing_upper_l_05 {parent=dyn_wing_lower_l_MSTR, loc=<-10.9451 -1.8257 -0.4175>, rot=<0 0 -0.039 -0.9992>}
 dyn_wing_upper_l_05_end {parent=dyn_wing_upper_l_05, loc=<-8.0176 0 0>, rot=<-0.004 -0.0349 0.0426 -0.9985>}
 dfrm_lowerarm_twist_01_l {parent=dfrm_lowerarm_l, loc=<18.316 0 0>, rot=<0 0 0 1>}
 dfrm_hand_l {parent=dfrm_lowerarm_l, loc=<22.7355 -0.3203 -0.6105>, rot=<-0.6784 -0.0223 0.0735 -0.7307>}
 dfrm_backwing_l {parent=dfrm_upperarm_l, loc=<5.1487 -5.7421 -9.7448>, rot=<-0.0369 0.0824 0.9737 0.2091>}
 deform_ADS_upperarm_compensation {parent=dfrm_upperarm_l, loc=<6.9848 -4.876 1.8034>, rot=<0.4292 -0.0172 0.0628 -0.9009>}
 deform_sleeve_aim_compensation_fr_l {parent=dfrm_upperarm_l, loc=<10.1743 9.0886 -1.9532>, rot=<0 0 0 1>}
 deform_sleeve_aim_compensation_btm_l {parent=dfrm_upperarm_l, loc=<10.0626 -1.2847 -8.179>, rot=<-0.1865 -0.0317 0.1159 -0.9751>}
 dfrm_clavicle_r {parent=spine_05, loc=<5.4831 -0.1053 3.2946>, rot=<0.7289 0.0465 -0.6829 0.009>}
 dfrm_upperarm_r {parent=dfrm_clavicle_r, loc=<-9.8658 1.2136 0.4105>, rot=<0.0164 -0.3245 -0.0269 -0.9454>}
 dfrm_upperarm_twist_01_r {parent=dfrm_upperarm_r, loc=<-7.2135 0.1134 0>, rot=<0 0 0 -1>}
 dyn_wing_upper_r_twist_MSTR {parent=dfrm_upperarm_twist_01_r, loc=<0 0 0>, rot=<0 0 0 1>}
 dyn_wing_upper_r_MSTR {parent=dyn_wing_upper_r_twist_MSTR, loc=<-2.8583 3.3049 3.1373>, rot=<-0.0599 -0.3671 0.8988 0.2321>}
 dyn_wing_upper_r_01 {parent=dyn_wing_upper_r_MSTR, loc=<13.8171 2.7401 1.3022>, rot=<-0.0902 0.1199 -0.1208 -0.9813>}
 dyn_wing_upper_r_01_end {parent=dyn_wing_upper_r_01, loc=<6.8788 0 0>, rot=<0.0079 -0.0534 0.0092 -0.9985>}
 dyn_wing_upper_r_02 {parent=dyn_wing_upper_r_MSTR, loc=<7.1186 -1.3473 0.9034>, rot=<0 0 0 1>}
 dyn_wing_upper_r_02_end {parent=dyn_wing_upper_r_02, loc=<6.8685 0 0>, rot=<-0.0087 0.0178 -0.0208 -0.9996>}
 dyn_wing_upper_r_03 {parent=dyn_wing_upper_r_MSTR, loc=<1.5403 -4.0703 0.5142>, rot=<0.0076 0.0232 0.1055 -0.9941>}
 dyn_wing_upper_r_03_end {parent=dyn_wing_upper_r_03, loc=<6.0908 0 0>, rot=<-0.0222 0.0342 -0.0093 -0.9991>}
 dfrm_upperarm_twist_02_r {parent=dfrm_upperarm_r, loc=<-19.7998 0.1134 0>, rot=<0 0 0 -1>}
 dfrm_lowerarm_r {parent=dfrm_upperarm_r, loc=<-26.465 0.8345 -0.259>, rot=<0.0036 0.0161 -0.2298 -0.9731>}
 dfrm_lowerarm_twist_02_r {parent=dfrm_lowerarm_r, loc=<-10.6707 0 0>, rot=<0 0 0 -1>}
 dyn_wing_lower_r_twist_MSTR {parent=dfrm_lowerarm_twist_02_r, loc=<0 0 0>, rot=<0 0 0 1>}
 dyn_wing_lower_r_MSTR {parent=dyn_wing_lower_r_twist_MSTR, loc=<7.6577 2.0482 0.4387>, rot=<0.0097 -0.1651 0.9668 0.1948>}
 dyn_wing_upper_r_04 {parent=dyn_wing_lower_r_MSTR, loc=<1.1459 -3.2663 0.3851>, rot=<0.0433 -0.0165 -0.0185 -0.9988>}
 dyn_wing_upper_r_04_end {parent=dyn_wing_upper_r_04, loc=<9.3845 0 0>, rot=<0.0023 0.0263 0.0172 -0.9995>}
 dyn_wing_upper_r_05 {parent=dyn_wing_lower_r_MSTR, loc=<10.9446 1.8259 0.417>, rot=<0 0 -0.039 -0.9992>}
 dyn_wing_upper_r_05_end {parent=dyn_wing_upper_r_05, loc=<8.0175 0 0>, rot=<-0.004 -0.0349 0.0426 -0.9985>}
 dfrm_lowerarm_twist_01_r {parent=dfrm_lowerarm_r, loc=<-18.316 0 0>, rot=<0 0 0 -1>}
 dfrm_hand_r {parent=dfrm_lowerarm_r, loc=<-22.7355 0.3203 0.6106>, rot=<-0.6784 -0.0223 0.0735 -0.7307>}
 dfrm_backwing_r {parent=dfrm_upperarm_r, loc=<-5.1486 5.7421 9.7447>, rot=<-0.0369 0.0824 0.9737 0.2091>}
 deform_sleeve_aim_compensation_fr_r {parent=dfrm_upperarm_r, loc=<-5.7603 -7.156 3.6285>, rot=<0 0 0 1>}
 deform_sleeve_aim_compensation_btm_r {parent=dfrm_upperarm_r, loc=<-11.213 1.3289 8.1983>, rot=<-0.2433 -0.0146 0.0583 -0.9681>}
 deform_collar_r {parent=spine_05, loc=<7.5461 6.7835 9.9175>, rot=<0.6791 0.197 0.6791 -0.197>}
 dyn_collar_r_MSTR {parent=deform_collar_r, loc=<1.9864 -0.4149 -0.5931>, rot=<0 0 0 1>}
 dyn_collar_r_1 {parent=dyn_collar_r_MSTR, loc=<0 0 0>, rot=<0 0 0 1>}
 dyn_collar_r_2 {parent=dyn_collar_r_1, loc=<2.3619 1.1246 -1.1631>, rot=<-0.0805 -0.1695 0.0012 -0.9822>}
 deform_collar_l {parent=spine_05, loc=<7.5457 6.7834 -9.9175>, rot=<0.197 -0.6791 -0.197 -0.6791>}
 dyn_collar_l_MSTR {parent=deform_collar_l, loc=<1.7976 0.5722 -0.6913>, rot=<0.0496 0 0 -0.9988>}
 dyn_collar_l_1 {parent=dyn_collar_l_MSTR, loc=<0 0 0>, rot=<0 0 0 1>}
 dyn_collar_l_2 {parent=dyn_collar_l_1, loc=<3.6666 -2.6314 0.4909>, rot=<-0.04 0.1333 -0.2849 -0.9484>}
 dyn_shirt_upper_fr_l_MSTR {parent=spine_04, loc=<0.4708 7.6849 -13.1124>, rot=<-0.6853 -0.7018 -0.185 0.0604>}
 dyn_shirt_upper_fr_l {parent=dyn_shirt_upper_fr_l_MSTR, loc=<0 0 0>, rot=<0 0 0 1>}
 dyn_shirt_upper_fr_r_MSTR {parent=spine_04, loc=<0.4704 7.6849 13.1124>, rot=<0.0604 -0.185 -0.7018 -0.6853>}
 dyn_shirt_upper_fr_r {parent=dyn_shirt_upper_fr_r_MSTR, loc=<0 0 0>, rot=<0 0 0 1>}
 deform_armpit_bk_l {parent=spine_04, loc=<1.1615 -10.2186 -7.3701>, rot=<0.5581 0.5581 -0.4342 0.4342>}
 deform_armpit_bk_r {parent=spine_04, loc=<1.1618 -10.2186 7.3701>, rot=<-0.4342 -0.4342 0.5581 -0.5581>}
 dyn_shirt_bk_l {parent=spine_03, loc=<-1.7175 -5.5402 -13.5926>, rot=<-0.6332 -0.7363 0.1411 -0.1922>}
 dyn_shirt_fr_l {parent=spine_03, loc=<-2.9873 9.2174 -14.3678>, rot=<-0.6838 -0.7082 -0.1747 0.0163>}
 dyn_shirt_bk {parent=spine_03, loc=<-1.1886 -11.9538 0>, rot=<-0.4934 -0.5065 0.4934 -0.5065>}
 dyn_shirt_bk_r {parent=spine_03, loc=<-1.7174 -5.5402 13.5926>, rot=<-0.1411 -0.1922 0.6332 -0.7363>}
 dyn_shirt_fr_r {parent=spine_03, loc=<-2.9872 9.2174 14.3678>, rot=<0.1747 0.0163 0.6838 -0.7082>}
 dfrm_pelvis {parent=pelvis, loc=<-12.6365 0 0>, rot=<0 0 0 1>}
 dfrm_thigh_r {parent=dfrm_pelvis, loc=<-3.2425 0.2481 8.7122>, rot=<0.0572 -0.0012 0.9982 0.0155>}
 dfrm_calf_r {parent=dfrm_thigh_r, loc=<37.1895 -0.1455 0.06>, rot=<0.0006 -0.0161 -0.0391 -0.9991>}
 dfrm_foot_r {parent=dfrm_calf_r, loc=<42.065 -0.945 -1.8502>, rot=<-0.0025 -0.0618 -0.6802 -0.7304>}
 dyn_ball_r {parent=dfrm_foot_r, loc=<-5.8402 -1.077 0.421>, rot=<0 0 0 1>}
 dyn_toe_r_c_01 {parent=dyn_ball_r, loc=<0.5132 -0.201 3.0849>, rot=<-0.2023 -0.6858 -0.6661 0.2124>}
 dyn_toe_r_c_02 {parent=dyn_toe_r_c_01, loc=<3.8303 0 0>, rot=<0 -0.0255 0.0482 -0.9985>}
 joint22 {parent=dyn_toe_r_c_02, loc=<5.0692 0 0>, rot=<0 0 0 1>}
 dyn_toe_r_b_01 {parent=dyn_ball_r, loc=<-0.702 -0.157 0.2106>, rot=<0.0098 0.7004 0.7136 0.0112>}
 dyn_toe_r_b_02 {parent=dyn_toe_r_b_01, loc=<5.8662 0 0>, rot=<0 -0.0345 0.0155 -0.9993>}
 joint23 {parent=dyn_toe_r_b_02, loc=<5.7004 0 0>, rot=<0 0 0 1>}
 dyn_toe_r_a_01 {parent=dyn_ball_r, loc=<0.3668 -0.1038 -2.6579>, rot=<-0.156 0.673 0.6981 0.188>}
 dyn_toe_r_a_02 {parent=dyn_toe_r_a_01, loc=<3.64 0 0>, rot=<0 -0.0797 -0.0135 -0.9967>}
 joint24 {parent=dyn_toe_r_a_02, loc=<4.3265 0 0>, rot=<0 0 0 1>}
 dyn_toe_r_d_03 {parent=dfrm_foot_r, loc=<2.6964 0.0647 0.4623>, rot=<0.7071 -0.1005 0.0447 -0.6985>}
 dyn_toe_r_d_01 {parent=dyn_toe_r_d_03, loc=<2.0486 0 0>, rot=<-0.0333 -0.0352 0.0011 -0.9988>}
 dyn_toe_r_d_02 {parent=dyn_toe_r_d_01, loc=<1.8873 0 0>, rot=<0 0.066 0 -0.9978>}
 joint15 {parent=dyn_toe_r_d_02, loc=<3.4635 0 0>, rot=<0 0 0 1>}
 dfrm_calf_twist_01_r {parent=dfrm_calf_r, loc=<29.2812 0.7545 -0.3098>, rot=<0.0237 0.0095 -0.0035 -0.9997>}
 dfrm_calf_twist_02_r {parent=dfrm_calf_r, loc=<12.2745 0.7545 -0.3098>, rot=<0.0237 0.0095 -0.0035 -0.9997>}
 deform_knee_in_base_r {parent=dfrm_calf_r, loc=<0.3084 7.9762 -0.8686>, rot=<0 0 0 -1>}
 deform_knee_in_r {parent=deform_knee_in_base_r, loc=<0 0 0>, rot=<0 0 0 1>}
 deform_knee_out_base_r {parent=dfrm_calf_r, loc=<-0.8551 -8.4142 -0.7848>, rot=<0 0 0 -1>}
 deform_knee_out_r {parent=deform_knee_out_base_r, loc=<0 0 0>, rot=<0 0 0 1>}
 dfrm_thigh_twist_01_r {parent=dfrm_thigh_r, loc=<13.2028 0 0>, rot=<0.0013 0 0 -1>}
 deform_groin_base_r {parent=dfrm_thigh_r, loc=<-1.1526 -11.1928 1.1578>, rot=<0 0 0 -1>}
 deform_groin_r {parent=deform_groin_base_r, loc=<0 0 0>, rot=<0 0 0 -1>}
 dfrm_thigh_l {parent=dfrm_pelvis, loc=<-3.2425 0.2481 -8.7122>, rot=<0.0012 0.0572 0.0155 -0.9982>}
 dfrm_calf_l {parent=dfrm_thigh_l, loc=<-37.1895 0.1455 -0.06>, rot=<0.0006 -0.0161 -0.0391 -0.9991>}
 dfrm_foot_l {parent=dfrm_calf_l, loc=<-42.065 0.945 1.8503>, rot=<-0.0025 -0.0618 -0.6802 -0.7304>}
 dyn_ball_l {parent=dfrm_foot_l, loc=<5.8402 1.077 -0.421>, rot=<0 0 0 1>}
 dyn_toe_l_b_01 {parent=dyn_ball_l, loc=<0.7011 0.1507 -0.1672>, rot=<-0.7053 0.0209 0.0214 -0.7083>}
 dyn_toe_l_b_02 {parent=dyn_toe_l_b_01, loc=<7.3191 0 0>, rot=<0.0132 -0.0869 -0.0211 -0.9959>}
 joint6 {parent=dyn_toe_l_b_02, loc=<4.2827 0 0>, rot=<0 0 0 1>}
 dyn_toe_l_a_01 {parent=dyn_ball_l, loc=<-0.4612 0.1551 -3.2374>, rot=<-0.6727 -0.187 0.1986 -0.6878>}
 dyn_toe_l_a_02 {parent=dyn_toe_l_a_01, loc=<4.8931 0 0>, rot=<-0.1023 -0.0736 -0.0192 -0.9918>}
 joint3 {parent=dyn_toe_l_a_02, loc=<3.8979 0 0>, rot=<0 0 0 1>}
 dyn_toe_l_c_01 {parent=dyn_ball_l, loc=<-0.3668 0.1038 2.6579>, rot=<-0.6896 0.1883 -0.1422 -0.6847>}
 dyn_toe_l_c_02 {parent=dyn_toe_l_c_01, loc=<5.0037 0 0>, rot=<0.0036 -0.1325 -0.0096 -0.9911>}
 joint9 {parent=dyn_toe_l_c_02, loc=<3.0034 0 0>, rot=<-0.5168 -0.8206 -0.13 0.2064>}
 dyn_toe_l_d_03 {parent=dfrm_foot_l, loc=<-2.6963 -0.0647 -0.4626>, rot=<-0.0696 -0.6943 0.7053 0.1251>}
 dyn_toe_l_d_01 {parent=dyn_toe_l_d_03, loc=<2.0436 0 0.1444>, rot=<0 0 0 1>}
 dyn_toe_l_d_02 {parent=dyn_toe_l_d_01, loc=<1.8873 0 0>, rot=<-0.0107 0.0658 -0.0051 -0.9978>}
 joint12 {parent=dyn_toe_l_d_02, loc=<3.4635 0 0>, rot=<0 0 0 1>}
 dfrm_calf_twist_01_l {parent=dfrm_calf_l, loc=<-29.2811 -0.7545 0.3098>, rot=<0.0237 0.0095 -0.0035 -0.9997>}
 dfrm_calf_twist_02_l {parent=dfrm_calf_l, loc=<-12.2744 -0.7545 0.3098>, rot=<0.0237 0.0095 -0.0035 -0.9997>}
 deform_knee_in_base_l {parent=dfrm_calf_l, loc=<-0.3084 -7.9762 0.8686>, rot=<0 0 0 -1>}
 deform_knee_in_l {parent=deform_knee_in_base_l, loc=<0 0 0>, rot=<0 0 0 1>}
 deform_knee_out_base_l {parent=dfrm_calf_l, loc=<0.8551 8.4142 0.7849>, rot=<0 0 0 -1>}
 deform_knee_out_l {parent=deform_knee_out_base_l, loc=<0 0 0>, rot=<0 0 0 1>}
 dfrm_thigh_twist_01_l {parent=dfrm_thigh_l, loc=<-13.2027 0 0>, rot=<0.0013 0 0 -1>}
 deform_groin_base_l {parent=dfrm_thigh_l, loc=<1.1526 11.1928 -1.1578>, rot=<0 0 0 -1>}
 deform_groin_l {parent=deform_groin_base_l, loc=<0 0 0>, rot=<0 0 0 1>}
 deform_glute_base_r {parent=dfrm_pelvis, loc=<-8.4059 -13.215 8.4287>, rot=<0.0501 -0.0275 0.8926 -0.4472>}
 deform_glute_r {parent=deform_glute_base_r, loc=<0 0 0>, rot=<0 0 0 1>}
 deform_glute_base_l {parent=dfrm_pelvis, loc=<-8.4058 -13.215 -8.4287>, rot=<0.0275 0.0501 -0.4472 -0.8926>}
 deform_glute_l {parent=deform_glute_base_l, loc=<0 0 0>, rot=<0 0 0 1>}
 thigh_l {parent=pelvis, loc=<-1.0855 -0.2109 -8.7053>, rot=<0.0009 0.0355 0.0155 -0.9993>}
 calf_l {parent=thigh_l, loc=<-42.5284 -0.0019 -0.7297>, rot=<0.0006 -0.0161 -0.0391 -0.9991>}
 dfrm_calf_l_PV {parent=calf_l, loc=<-5.4607 50.5823 -0.0167>, rot=<0.0017 0.0216 0.0001 -0.9998>}
 PV L {controlType=EULER_TRANSFORM, parent=calf_l, type=Control, loc=<0 50 0>, rot=<0 0 0 1>}
 dfrm_calf_aim_l {parent=calf_l, loc=<-40.7601 -20.4095 -0.4263>, rot=<0.0017 0.0216 0.0001 -0.9998>}
 foot_l {parent=calf_l, loc=<-46.7401 -2.5014 -0.592>, rot=<0.0427 -0.0238 0.0353 -0.9982>}
 ball_l {parent=foot_l, loc=<-3.5658 10.8066 -0.1544>, rot=<0.0001 -0.0001 -0.7063 -0.7079>}
 dfrm_foot_l_ctrl {controlType=EULER_TRANSFORM, parent=foot_l, type=Control, loc=<14.6217 -3.4245 3.2306>, rot=<-0.5229 -0.4735 0.4801 0.5214>}
 calf_twist_02_l {parent=calf_l, loc=<-19.1546 -1.7943 -0.2638>, rot=<0.0237 0.0095 -0.0035 -0.9997>}
 calf_twist_01_l {parent=calf_l, loc=<-35.8483 -1.9197 -0.5794>, rot=<0.0237 0.0095 -0.0035 -0.9997>}
 thigh_twist_01_l {parent=thigh_l, loc=<-27.952 0 0>, rot=<0.0013 0 0 -1>}
 thigh_r {parent=pelvis, loc=<-1.0855 -0.2109 8.7053>, rot=<0.0355 -0.0009 0.9993 0.0155>}
 calf_r {parent=thigh_r, loc=<42.5284 0.0019 0.7297>, rot=<0.0006 -0.0161 -0.0391 -0.9991>}
 dfrm_calf_r_PV {parent=calf_r, loc=<5.4562 -50.5821 0.0765>, rot=<0.0017 0.0216 0.0001 -0.9998>}
 PV R {controlType=EULER_TRANSFORM, parent=calf_r, type=Control, loc=<0 -50 0>, rot=<0 0 0 1>}
 dfrm_calf_aim_r {parent=calf_r, loc=<42.7689 20.0655 0.5145>, rot=<0.0017 0.0216 0.0001 -0.9998>}
 foot_r {parent=calf_r, loc=<46.7401 2.5014 0.592>, rot=<0.0427 -0.0238 0.0353 -0.9982>}
 ball_r {parent=foot_r, loc=<3.5658 -10.8066 0.1544>, rot=<0.0001 -0.0001 -0.7063 -0.7079>}
 dfrm_foot_r_ctrl {controlType=EULER_TRANSFORM, parent=foot_r, type=Control, loc=<0 -1 0>, rot=<0 0 0 1>}
 calf_twist_02_r {parent=calf_r, loc=<19.1546 1.7943 0.2638>, rot=<0.0237 0.0095 -0.0035 -0.9997>}
 calf_twist_01_r {parent=calf_r, loc=<35.8483 1.9197 0.5794>, rot=<0.0237 0.0095 -0.0035 -0.9997>}
 thigh_twist_01_r {parent=thigh_r, loc=<27.952 0 0>, rot=<0.0013 0 0 -1>}
 Foot_l_COL {parent=root, loc=<0 0 0>, rot=<0 0 0 1>}
 Foot_r_COL {parent=root, loc=<0 0 0>, rot=<0 0 0 1>}
 Floor_COL {parent=root, loc=<0 0 0>, rot=<0 0 0 1>}
]
variables [
 SPR_upperarm_l_down_A_alpha {type=double, default=0}
 SPR_upperarm_r_down_A_alpha {type=double, default=0}
 SPR_upperarm_r_down_B_alpha {type=double, default=0}
 deform_bool {
  type=FUserDefinedStruct_#4
  default="(upperarm_1_#5=True,elbow_3_#6=True,wrist_7_#7=True,groin_8_#3=True,glute_17_#2=True,knee_9_40C408324CA261 ...(truncated)"
 }
 deform_value {
  type=FUserDefinedStruct_#8
  default="(upperarm_up_18_#9=(Rotation=(X=0,Y=0,Z=0,W=1),Translation=(X=0,Y=0,Z=8),Scale3D=(X=0.7,Y=1,Z=1)),upperarm_fwd_fr_19_#10=(Rotation=(X=0,Y=0,Z=0,W=1),Translation=(X=0,Y=10,Z=0),Scale3D=(X=0.5,Y=1,Z= ...(truncated)"
 }
 isEmote {type=bool, default=False}
 isFrontend {type=bool, default=False}
 mult_spr_dst_hand_r_thigh_twist_r_alpha {type=double, default=0}
 mult_spr_dst_lowerarm_r_pelvis_alpha {type=double, default=0}
 thigh r up {type=double, default=0}
]
