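# Physics asset for the FaceAcc mesh part of the female medium SunnySquawk character, set up for dynamic shirt bones against the upper body.
# Spheres on the spine and capsules on the clavicles and upper-arm twist bones form the body volumes, with smaller spheres on each dyn_shirt bone.
# Back and lower front shirt bones constrain to spine_03 with narrow angular limits.
# Upper front bones constrain to spine_04 with only swing2 given range.
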PhysicsAsset /BRCosmetics/Characters/Player/Female/Medium/Bodies/F_MED_SunnySquawk/Meshes/Parts/F_MED_SunnySquawk_FaceAcc_Physics
bodies {
 dfrm_clavicle_l [{shapeType=Capsule, length=13.5, radius=6, shapeName=dfrm_clavicle_l_capsule, center=<12.0921 0 -3.0106>, rotation=<rot 90 -179.9801 0.0193>}]
 dfrm_clavicle_r [{
   shapeType=Capsule
   length=14
   radius=5.5
   shapeName=dfrm_clavicle_r_capsule
   center=<-12.0921 0 3.0106>
   rotation=<rot 90 -152.0616 -152.0621>
  }]
 dfrm_upperarm_twist_01_l [{shapeType=Capsule, length=15.5, radius=7.5, shapeName=dfrm_upperarm_twist_01_l_capsule, center=<4.684 0 -0.7069>, rotation=<rot 90 0 0>}]
 dfrm_upperarm_twist_01_r [{shapeType=Capsule, length=15, radius=8, shapeName=dfrm_upperarm_twist_01_r_capsule, center=<-4.684 0 -0.3132>, rotation=<rot 90 0 0>}]
 dyn_shirt_bk [{shapeType=Sphere, radius=5, shapeName=dyn_shirt_bk_sphere, center=<0 -4.3031 5.0164>}]
 dyn_shirt_bk_l [{shapeType=Sphere, radius=5, shapeName=dyn_shirt_bk_l_sphere, center=<0 -4.3031 3.9712>}]
 dyn_shirt_bk_r [{shapeType=Sphere, radius=5, shapeName=dyn_shirt_bk_r_sphere, center=<0 -4.3031 4.1542>}]
 dyn_shirt_fr_l [{shapeType=Sphere, radius=5, shapeName=dyn_shirt_fr_l_sphere, center=<0 -5.5817 2.7294>}]
 dyn_shirt_fr_r [{shapeType=Sphere, radius=5, shapeName=dyn_shirt_fr_r_sphere, center=<0 -6.3622 3.1257>}]
 dyn_shirt_upper_fr_l [{shapeType=Sphere, radius=3.2069, shapeName=dyn_shirt_upper_fr_l_sphere, center=<5.4744 -3.7777 1.1496>}]
 dyn_shirt_upper_fr_r [{shapeType=Sphere, radius=3.3162, shapeName=dyn_shirt_upper_fr_r_sphere, center=<5.6201 3.6324 1.4294>}]
 spine_02 [{shapeType=Sphere, radius=15.5, shapeName=spine_02_sphere, center=<0 3.44 0>}]
 spine_03 [{shapeType=Sphere, radius=15, shapeName=spine_03_sphere, center=<0 3.44 0>}]
 spine_04 [{shapeType=Sphere, radius=12.5, shapeName=spine_04_sphere, center=<0 1.4415 0>}]
 spine_05 [{shapeType=Sphere, radius=18.5, shapeName=spine_05_sphere, center=<-3.6181 0 0>}]
}
constraints [
 {bone1Name=spine_04, bone2Name=spine_03, swing1LimitDegrees=45, swing2LimitDegrees=45, twistLimitDegrees=45}
 {bone1Name=spine_05, bone2Name=spine_04, swing1LimitDegrees=45, swing2LimitDegrees=45, twistLimitDegrees=45}
 {bone1Name=dfrm_clavicle_l, bone2Name=spine_05, swing1LimitDegrees=45, swing2LimitDegrees=45, twistLimitDegrees=45}
 {
  bone1Name=spine_03
  bone2Name=spine_02
  swing1LimitDegrees=45
  swing1Motion=Free
  swing2LimitDegrees=45
  swing2Motion=Free
  twistLimitDegrees=45
  twistMotion=Free
 }
 {bone1Name=dyn_shirt_bk, bone2Name=spine_03, swing1LimitDegrees=5, swing2LimitDegrees=30, twistLimitDegrees=5}
 {bone1Name=dyn_shirt_bk_l, bone2Name=spine_03, swing1LimitDegrees=5, swing2LimitDegrees=5, twistLimitDegrees=5}
 {bone1Name=dyn_shirt_bk_r, bone2Name=spine_03, swing1LimitDegrees=5, swing2LimitDegrees=5, twistLimitDegrees=5}
 {bone1Name=dyn_shirt_fr_l, bone2Name=spine_03, swing1LimitDegrees=5, swing2LimitDegrees=30, twistLimitDegrees=5}
 {bone1Name=dyn_shirt_fr_r, bone2Name=spine_03, swing1LimitDegrees=5, swing2LimitDegrees=30, twistLimitDegrees=5}
 {bone1Name=dyn_shirt_upper_fr_l, bone2Name=spine_04, swing1LimitDegrees=0, swing2LimitDegrees=20, twistLimitDegrees=0}
 {bone1Name=dyn_shirt_upper_fr_r, bone2Name=spine_04, swing1LimitDegrees=0, swing2LimitDegrees=20, twistLimitDegrees=0}
 {
  bone1Name=dfrm_clavicle_r
  bone2Name=spine_05
  swing1LimitDegrees=45
  swing1Motion=Free
  swing2LimitDegrees=45
  swing2Motion=Free
  twistLimitDegrees=45
  twistMotion=Free
 }
 {
  bone1Name=dfrm_upperarm_twist_01_l
  bone2Name=dfrm_clavicle_l
  swing1LimitDegrees=45
  swing1Motion=Free
  swing2LimitDegrees=45
  swing2Motion=Free
  twistLimitDegrees=45
  twistMotion=Free
 }
 {
  bone1Name=dfrm_upperarm_twist_01_r
  bone2Name=dfrm_clavicle_r
  swing1LimitDegrees=45
  swing1Motion=Free
  swing2LimitDegrees=45
  swing2Motion=Free
  twistLimitDegrees=45
  twistMotion=Free
 }
]
