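# Animation Blueprint for the SunnySquawk female character's face accessory part.
# It copies the attached parent's pose and curves, runs a Control Rig fed front-end and emote flags, then adds pelvis-relative RigidBody physics with damped component motion.
# On customization load it caches part meshes.
# Each update, Do Once gates that reset each other apply secondary-layer skin weight profiles when a shoe (ExtraPart) mesh exists, and clear them otherwise.
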AnimBlueprint /BRCosmetics/Characters/Player/Female/Medium/Bodies/F_MED_SunnySquawk/Meshes/Parts/F_MED_SunnySquawk_FaceAcc_AnimBP
target_skeleton=&/F_MED_SunnySquawk_FaceAcc_Skeleton
graphs [
 AnimGraph {nodes [
   Component To Local()
   n1 = Control Rig!(
    Source=Copy Pose From Mesh() {bCopyCurves=true, bUseAttachedParent=true}
    Alpha
    isFrontend:bool=Get bIsInFrontEnd()
    isEmote:bool=Get bIsPlayingEmote()
   ) -> (Pose) {
    alphaBoolBlend {blendOption=Linear}
    controlRigAssetReference {blueprintRigClass=&/F_MED_SunnySquawk_FaceAcc_CtrlRig.F_MED_SunnySquawk_FaceAcc_CtrlRig_C}
   }
   n2 = RigidBody!(ComponentPose=Local To Component(LocalPose=@n1), Alpha, OverrideWorldGravity:Vector="0,0,-980", ExternalForce) -> (Pose) {
    bOverrideWorldGravity=true
    overridePhysicsAsset=&/F_MED_SunnySquawk_FaceAcc_Physics
    simulationSpace=BaseBoneSpace
    alphaBoolBlend {blendOption=Linear}
    alphaScaleBiasClamp {interpSpeedDecreasing=5, interpSpeedIncreasing=7}
    baseBoneRef {boneName=pelvis}
    componentAppliedLinearAccClamp=<3000 3000 3000>
    componentLinearAccScale=<0.3 0.3 0.3>
    componentLinearVelScale=<0.3 0.3 0.3>
    simSpaceSettings {
     maxAngularAcceleration=100
     maxAngularVelocity=3
     maxLinearAcceleration=700
     maxLinearVelocity=500
     velocityScaleZ=0
     worldAlpha=1
     externalLinearDragV=<0.2 0.2>
    }
   }
   Output Pose(Result=Component To Local(ComponentPose=@n2)) {layerGroup=None}
  ]}
 EventGraph {nodes [
   Event BlueprintUpdateAnimation()
   > n1 = Get!(execute:Exec) -> (then:Exec, else:Exec, Shoe_Mesh:Skeletal Mesh Component Budgeted Object Reference)
   > n2 = Sequence()
   > Do Once(Reset=@n5.then_1)
   > n3 = Is Valid(InputObject=Get SkeletalMesh!(self:Skinned Mesh Component Object Reference=@n1) -> (SkeletalMesh:Skeletal Mesh Object Reference))
   ClearAllSkinWeightProfiles(
    execute=[@n4.Completed, @n3.Is Not Valid]
    self:Skinned Mesh Component Object Reference=Get FaceAcc_Mesh() -> (FaceAcc_Mesh:Skeletal Mesh Component Budgeted Object Reference)
   )
   SetSkinWeightProfile!(
    execute=@n3.Is Valid
    self:Skinned Mesh Component Object Reference=Get FaceAcc_Mesh()
    InProfileName:Name=Alternative
    InLayer:ESkinWeightProfileLayer Enum=Secondary
   ) -> (then, ReturnValue:bool)
   > SetSkinWeightProfile(self=Get FaceAcc_Mesh(), InProfileName=VertexCrunch, InLayer=Secondary)
   n4 = Do Once(execute=@n5.then_0, Reset=@n2.then_1)
   n5 = Sequence(execute=@n1.else)
   n6 = Event OnCharacterCustomizationLoaded()
   > n7 = Sequence()
   > Set Character Mesh(Character Mesh:Skeletal Mesh Component Object Reference=@n6.CustomBaseMesh) -> (Output_Get:Skeletal Mesh Component Object Reference)
   Set Body Mesh(
    execute=@n7.then_1
    Body Mesh:Skeletal Mesh Component Budgeted Object Reference=Get Array Item with EFort Custom Part Type(Array=@n6.CustomMeshParts)
   ) -> (Output_Get:Skeletal Mesh Component Budgeted Object Reference)
   > Set Shoe_Mesh(Shoe_Mesh:Skeletal Mesh Component Budgeted Object Reference=Get Array Item with EFort Custom Part Type(Array=@n6.CustomMeshParts, PartType=ExtraPart)) -> (Output_Get:Skeletal Mesh Component Budgeted Object Reference)
   > Set FaceAcc_Mesh(FaceAcc_Mesh:Skeletal Mesh Component Budgeted Object Reference=Get Array Item with EFort Custom Part Type(Array=@n6.CustomMeshParts, PartType=Face)) -> (Output_Get:Skeletal Mesh Component Budgeted Object Reference)
   GetSkeletalMesh!(self:Custom Character Part Object Reference=Get Array Item with EFort Custom Part Type(Array:Array of Custom Character Part Object References=@n6.CustomParts) -> (Output:Custom Character Part Object Reference)) -> (ReturnValue:Skeletal Mesh Object Reference)
  ]}
]
variables [
 Body Mesh {type=SkeletalMeshComponentBudgeted}
 Character Mesh {type=SkeletalMeshComponent}
 FaceAcc_Mesh {type=SkeletalMeshComponentBudgeted}
 Shoe_Mesh {type=SkeletalMeshComponentBudgeted}
]
